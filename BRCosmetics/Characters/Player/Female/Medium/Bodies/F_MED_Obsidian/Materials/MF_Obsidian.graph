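# Material function for the Obsidian character skin that layers an animated galaxy effect onto input material attributes and adds vertex wobble.
# Skin FX mask channels and a Fresnel-weighted panning noise, switchable between pre-skinned projection and UV sampling, feed MF_GalaxyObsidian.
# Its result gets black base color and specular.
# A time-driven sine over pre-skinned position offsets vertices in Z, fading out as vertex color green rises.
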MaterialFunction /BRCosmetics/Characters/Player/Female/Medium/Bodies/F_MED_Obsidian/Materials/MF_Obsidian
graph {nodes [
  n1 = PreSkinnedPosition()
  n2 = ShaderStageSwitch(PixelShader=VertexInterpolator(VS=@n1).PS, VertexShader=@n1)
  n3 = MaterialFunctionCall(Float3=Multiply(
    A=FunctionInput(Preview=@n2) {bUsePreviewValueAsDefault=true, inputName=Pre Skinned Position}
    B=ScalarParameter() {defaultValue=0.005, group=Obsidian, parameterName=Obsidian_WPO_LocalNoise_Size}
   ) {constB=0.02}) {materialFunction=&BreakOutFloat3Components}
  n4 = Add(A=Add(A=@n3.R, B=@n3.R), B=@n3.B)
  n5 = Multiply(
   A=Subtract(A=Time(), B=ScalarParameter() {group=Obsidian, parameterName=Obsidian_WPO_random_seed})
   B=ScalarParameter() {defaultValue=0.6, group=Obsidian, parameterName=Obsidian_WPO_LocalNoise_Speed}
  ) {constB=0.6}
  n6 = Sine(None=Subtract(A=Subtract(A=@n4) {constB=2}, B=Frac(None=@n5)))
  Add(A=@n6)
  n7 = FunctionInput() {inputType=FunctionInput_MaterialAttributes}
  n8 = TextureSample(Tex=TextureObjectParameter() {group=Skin FX, parameterName=SkinFX_Mask}) {samplerType=SAMPLERTYPE_Masks}
  n9 = StaticComponentMaskParameter(None=@n8.RGBA) {defaultR=true, group=Skin FX, parameterName=Galaxy Warp Mask Channel}
  n10 = VectorParameter() {group=Galaxy, parameterName=GalaxyEF, channelNames {a="", b=GspeedY, g=GspeedX, r=EdgeFresTile}, defaultValue=<rgba 100 0 0.1 1>}
  n11 = AppendVector(A=@n10.GspeedX, B=@n10.GspeedX)
  n12 = MaterialFunctionCall!(
   TextureObject=TextureObject() {samplerType=SAMPLERTYPE_LinearColor, texture=/Game/Weapons/WeaponSkins/Textures/T_Swirly_Noise_highContrast1}
   TextureSize=@n10.EdgeFresTile
   WorldPosition
   Export Float 4
   In_sideSpeed=VectorParameter() {parameterName=GalaxySideSpeed, defaultValue=<rgba 0.1 -0.05 0 0>}.RGB
   In=@n11
   World Space Normal
   ProjectionTransitionContrast
  ) {materialFunction=&/Preskinned_WP_Texture_ScalePan_obsidian}
  n13 = TextureCoordinate() {uTiling=2, vTiling=2}
  n14 = TextureSample(UVs=Panner(Coordinate=Multiply(A=@n13, B=@n10.EdgeFresTile), Speed=@n11)) {
   samplerSource=SSM_Wrap_WorldGroupSettings
   samplerType=SAMPLERTYPE_LinearColor
   texture=&T_Swirly_Noise_highContrast1
  }
  n15 = ScalarParameter() {defaultValue=6, group=Galaxy, parameterName=Galaxy_FresEX}
  n16 = Multiply(
   A=StaticSwitchParameter(True=ComponentMask(None=@n12.XYZ Texture) {r=true}, False=@n14.R) {defaultValue=true, group=Galaxy, parameterName=UseLocalEdgeFresnel}
   B=Saturate(None=Fresnel(ExponentIn=@n15) {baseReflectFraction=0})
  )
  n17 = StaticComponentMaskParameter(None=@n8.RGBA) {defaultR=true, group=Skin FX, parameterName=Galaxy Channel}
  n18 = Constant3Vector()
  n19 = Constant()
  n20 = MaterialFunctionCall(
   X=@n19
   Y=@n19
   Z=Multiply(A=@n6, B=ScalarParameter() {defaultValue=2, group=Obsidian, parameterName=Obsidian_WPO_LocalNoise_amplitude}) {constB=0.5}
  ) {materialFunction=&MakeFloat3}
  n21 = VertexColor()
  FunctionOutput(None=SetMaterialAttributes!(
    MaterialAttributes=MaterialFunctionCall!(In=@n7, Warp Mask=@n9, GalaxyAlternateFresnel=@n16, Alpha=@n17) {materialFunction=&/MF_GalaxyObsidian}.Result
    BaseColor=@n18
    Specular=@n18
    WorldPositionOffset=LinearInterpolate(A=@n20.Result, B=@n18, Alpha=@n21.G)
   ) {attributeSetTypes [#1, #2, #3]})
  MaterialFunctionCall!(In, Warp Mask, GalaxyAlternateFresnel, Alpha) {materialFunction=/Game/Weapons/WeaponSkins/MaterialFunction/Galaxy}
 ]}
