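# Master material for the Obsidian female medium player body cosmetic, a Default Lit skeletal mesh material that also supports mesh deformers.
# It uses material attributes, passing the shared character PreFX result through MF_Obsidian into the shared PostFX function.
# The PreFX AO output bypasses MF_Obsidian and drives the PostFX AO, Distance Blend, Fresnel and Gmap inputs alike.
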Material /BRCosmetics/Characters/Player/Female/Medium/Bodies/F_MED_Obsidian/Materials/M_Obsidian_Master
graph {
 properties {
  bUseMaterialAttributes=true
  bUsedWithMeshDeformer=true
  bUsedWithSkeletalMesh=true
  pixelDepthOffsetMode=PDOM_Legacy
  usedShadingModels=Default Lit
 }
 nodes [
  n1 = MaterialFunctionCall() {materialFunction=&BaseFN_CharacterFunction_PreFX}
  MaterialFunctionCall(
   AO=@n1.AO
   Distance Blend=@n1.AO
   Fresnel=@n1.AO
   Gmap=@n1.AO
   Material=MaterialFunctionCall!(Pre Skinned Position, In=@n1.Result) {materialFunction=&/MF_Obsidian}.Result
  ) {materialFunction=&BaseFN_CharacterFunction_PostFX}
 ]
}
