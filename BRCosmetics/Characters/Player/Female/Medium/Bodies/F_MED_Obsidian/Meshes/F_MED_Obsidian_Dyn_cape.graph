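# Physics asset for the dynamic cape and feathers on the medium female Obsidian player cosmetic.
# Three six-capsule cape chains hang from pins on both clavicles and upper spine, with swing limits widening below the top and tight twist.
# Angularly free cross-constraints tie the chains at rows 1, 3 and 4.
# Six limited feathers hang from master anchors, alongside body capsules, a crouch-collision sphere and a root slope box.
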PhysicsAsset /BRCosmetics/Characters/Player/Female/Medium/Bodies/F_MED_Obsidian/Meshes/F_MED_Obsidian_Dyn_cape
bodies {
 calf_l [{shapeType=Capsule, length=31.8125, radius=12.5363, center=<-31.4163 -2.6359 -0.0003>, rotation=<rot 90 0 -1.5716>}]
 calf_r [{shapeType=Capsule, length=31.8125, radius=12.5363, center=<31.4163 2.6359 0.0003>, rotation=<rot 90 0 178.4284>}]
 clavicle_l [{shapeType=Capsule, length=6.125, radius=8.005, center=<6.4728 0.344 -3.7639>, rotation=<rot 83.0162 10.3785 6.1841>}]
 clavicle_r [{shapeType=Capsule, length=6.125, radius=8.005, center=<-6.4728 -0.344 3.7639>, rotation=<rot 83.0161 10.3785 -173.8159>}]
 dyn_cape_l_01 [{shapeType=Capsule, length=13.2812, radius=15, center=<9.6403 0 15>, rotation=<rot -90 -0.0033 0.0033>}]
 dyn_cape_l_02 [{shapeType=Capsule, length=13.2812, radius=15, center=<9.6403 0 15>, rotation=<rot -90 0.0396 -0.0396>}]
 dyn_cape_l_03 [{shapeType=Capsule, length=13.2812, radius=15, center=<9.6403 0 15>, rotation=<rot -90 0.0396 -0.0396>}]
 dyn_cape_l_04 [{shapeType=Capsule, length=13.2812, radius=15, center=<9.6403 0 15>, rotation=<rot -90 0.0396 -0.0396>}]
 dyn_cape_l_05 [{shapeType=Capsule, length=13.2812, radius=15, center=<9.6403 0 15>, rotation=<rot -90 0.0396 -0.0396>}]
 dyn_cape_l_06 [{shapeType=Capsule, length=7.2188, radius=15, center=<7.0061 -0.0002 15>, rotation=<rot -90 0.0099 -0.0099>}]
 dyn_cape_mid_01 [{shapeType=Capsule, length=13.2812, radius=15, center=<9.6403 0 15>, rotation=<rot -90 0.0396 90>}]
 dyn_cape_mid_02 [{shapeType=Capsule, length=13.2812, radius=15, center=<9.6403 0 15>, rotation=<rot -90 0.0396 90>}]
 dyn_cape_mid_03 [{shapeType=Capsule, length=13.2812, radius=15, center=<9.6403 0 15>, rotation=<rot -90 0.0396 90>}]
 dyn_cape_mid_04 [{shapeType=Capsule, length=13.2812, radius=15, center=<9.6403 0 15>, rotation=<rot -90 0.0396 90>}]
 dyn_cape_mid_05 [{shapeType=Capsule, length=13.2812, radius=15, center=<9.6403 0 15>, rotation=<rot -90 0.0396 90>}]
 dyn_cape_mid_06 [{shapeType=Capsule, length=15.4688, radius=15, center=<7.0061 -0.0002 15>, rotation=<rot -90 0.0099 90>}]
 dyn_cape_pin_l_01 [{shapeType=Capsule, length=0.5, radius=0.5}]
 dyn_cape_pin_mid [{shapeType=Capsule, length=0.5, radius=0.5}]
 dyn_cape_pin_r_01 [{shapeType=Capsule, length=0.5, radius=0.5}]
 dyn_cape_r_01 [{shapeType=Capsule, length=13.2812, radius=15, center=<9.6403 0 15>, rotation=<rot -90 0.0396 -0.0396>}]
 dyn_cape_r_02 [{shapeType=Capsule, length=13.2812, radius=15, center=<9.6403 0 15>, rotation=<rot -90 0.0396 -0.0396>}]
 dyn_cape_r_03 [{shapeType=Capsule, length=13.2812, radius=15, center=<9.6403 0 15>, rotation=<rot -90 0.0396 -0.0396>}]
 dyn_cape_r_04 [{shapeType=Capsule, length=13.2812, radius=15, center=<9.6403 0 15>, rotation=<rot -90 0.0396 -0.0396>}]
 dyn_cape_r_05 [{shapeType=Capsule, length=13.2812, radius=15, center=<9.6403 0 15>, rotation=<rot -90 0.0396 -0.0396>}]
 dyn_cape_r_06 [{shapeType=Capsule, length=7.2188, radius=15, center=<7.0061 -0.0002 15>, rotation=<rot -90 0.0099 -0.0099>}]
 dyn_crouch_collision [{shapeType=Sphere, radius=53.3125, center=<-0.0006 5.0355 0>}]
 dyn_feather_bk_l [{shapeType=Capsule, length=15.6875, radius=7.9688, center=<5.7886 0.0003 0>, rotation=<rot -90 0.0099 -0.0099>}]
 dyn_feather_bk_l_MSTR [{shapeType=Capsule, length=0.5, radius=0.5}]
 dyn_feather_bk_r [{shapeType=Capsule, length=15.6875, radius=7.9688, center=<5.7886 0.0003 0>, rotation=<rot -90 0.0099 -0.0099>}]
 dyn_feather_bk_r_MSTR [{shapeType=Capsule, length=0.5, radius=0.5}]
 dyn_feather_fr_l [{shapeType=Capsule, length=15.6875, radius=7.9688, center=<5.7886 0.0003 0>, rotation=<rot -90 0.0099 -0.0099>}]
 dyn_feather_fr_l_MSTR [{shapeType=Capsule, length=0.5, radius=0.5}]
 dyn_feather_fr_r [{shapeType=Capsule, length=15.6875, radius=7.9688, center=<5.7886 0.0003 0>, rotation=<rot -90 0.0099 -0.0099>}]
 dyn_feather_fr_r_MSTR [{shapeType=Capsule, length=0.5, radius=0.5}]
 dyn_feather_l [{shapeType=Capsule, length=15.6875, radius=7.9688, center=<5.7886 0.0003 0>, rotation=<rot -90 0.0099 -0.0099>}]
 dyn_feather_l_MSTR [{shapeType=Capsule, length=0.5, radius=0.5}]
 dyn_feather_r [{shapeType=Capsule, length=15.6875, radius=7.9688, center=<5.7886 0.0003 0>, rotation=<rot -90 0.0099 -0.0099>}]
 dyn_feather_r_MSTR [{shapeType=Capsule, length=0.5, radius=0.5}]
 dyn_root_slope [{shapeType=Box, extentX=231.0625, extentY=231.0625, extentZ=12.375, center=<0 0 -5.9184>}]
 lowerarm_l [{shapeType=Capsule, length=31.9688, radius=8.63, center=<14.3724 1.1352 -1.668>, rotation=<rot 82.7121 -0.0002 3.2014>}]
 lowerarm_r [{shapeType=Capsule, length=31.9688, radius=8.63, center=<-14.3724 -1.1352 1.668>, rotation=<rot 82.7121 -0.0002 -176.7986>}]
 pelvis [
  {shapeType=Capsule, length=11.1562, radius=16.6925, center=<-1.7798 1.668 -0.0005>}
  {shapeType=Capsule, length=17.6562, radius=16.7238, center=<-16.3531 3.315 0.0007>}
 ]
 spine_01 [{shapeType=Capsule, length=11.5, radius=16.3488, center=<-1.5259 6.8717 -0.0002>}]
 spine_02 [{shapeType=Capsule, length=15.375, radius=12.4738, center=<0 5.1277 0>}]
 spine_03 [{shapeType=Capsule, length=15.2188, radius=13.1925, center=<0.0135 4.7511 0>}]
 spine_04 [{shapeType=Capsule, length=17.8125, radius=11.8488, center=<1.0291 2.9331 0.0002>}]
 spine_05 [{shapeType=Capsule, length=14.2188, radius=11.1613, center=<-2.6173 2.0295 -0.0006>}]
 thigh_l [{shapeType=Capsule, length=16.125, radius=18.5988, center=<-20.1925 4.0916 3.1216>, rotation=<rot 90 0 0>}]
 thigh_r [{shapeType=Capsule, length=16.125, radius=18.5988, center=<20.1925 -4.0916 -3.1216>, rotation=<rot 90 0 180>}]
 upperarm_l [
  {shapeType=Capsule, length=21.8438, radius=8.255, center=<13.1767 1.5602 -0.0001>, rotation=<rot 90 0 0>}
  {shapeType=Capsule, length=8.875, radius=7.875, center=<5.7274 0.6211 0.1463>, rotation=<rot -83.2014 -180 180>}
 ]
 upperarm_r [
  {shapeType=Capsule, length=21.8438, radius=8.255, center=<-13.1767 -1.5602 0.0001>, rotation=<rot 90 0 180>}
  {shapeType=Capsule, length=8.875, radius=7.875, center=<-5.7274 -0.6211 -0.1463>, rotation=<rot -83.2013 -180 0>}
 ]
}
constraints [
 {
  bone1Name=spine_01
  bone2Name=pelvis
  swing1LimitDegrees=45
  swing1Motion=Free
  swing2LimitDegrees=45
  swing2Motion=Free
  twistLimitDegrees=45
  twistMotion=Free
 }
 {
  bone1Name=spine_02
  bone2Name=spine_01
  swing1LimitDegrees=45
  swing1Motion=Free
  swing2LimitDegrees=45
  swing2Motion=Free
  twistLimitDegrees=45
  twistMotion=Free
 }
 {
  bone1Name=spine_03
  bone2Name=spine_02
  swing1LimitDegrees=45
  swing1Motion=Free
  swing2LimitDegrees=45
  swing2Motion=Free
  twistLimitDegrees=45
  twistMotion=Free
 }
 {
  bone1Name=spine_04
  bone2Name=spine_03
  swing1LimitDegrees=45
  swing1Motion=Free
  swing2LimitDegrees=45
  swing2Motion=Free
  twistLimitDegrees=45
  twistMotion=Free
 }
 {
  bone1Name=spine_05
  bone2Name=spine_04
  swing1LimitDegrees=45
  swing1Motion=Free
  swing2LimitDegrees=45
  swing2Motion=Free
  twistLimitDegrees=45
  twistMotion=Free
 }
 {
  bone1Name=clavicle_l
  bone2Name=spine_05
  swing1LimitDegrees=45
  swing1Motion=Free
  swing2LimitDegrees=45
  swing2Motion=Free
  twistLimitDegrees=45
  twistMotion=Free
 }
 {
  bone1Name=upperarm_l
  bone2Name=clavicle_l
  swing1LimitDegrees=45
  swing1Motion=Free
  swing2LimitDegrees=45
  swing2Motion=Free
  twistLimitDegrees=45
  twistMotion=Free
 }
 {
  bone1Name=lowerarm_l
  bone2Name=upperarm_l
  swing1LimitDegrees=45
  swing1Motion=Free
  swing2LimitDegrees=45
  swing2Motion=Free
  twistLimitDegrees=45
  twistMotion=Free
 }
 {
  bone1Name=dyn_cape_pin_r_01
  bone2Name=clavicle_r
  swing1LimitDegrees=45
  swing1Motion=Free
  swing2LimitDegrees=45
  swing2Motion=Free
  twistLimitDegrees=45
  twistMotion=Free
 }
 {bone1Name=dyn_cape_r_01, bone2Name=dyn_cape_pin_r_01, swing1LimitDegrees=10, swing2LimitDegrees=10, twistLimitDegrees=5}
 {bone1Name=dyn_cape_r_02, bone2Name=dyn_cape_r_01, swing1LimitDegrees=25, swing2LimitDegrees=25, twistLimitDegrees=5}
 {bone1Name=dyn_cape_r_03, bone2Name=dyn_cape_r_02, swing1LimitDegrees=35, swing2LimitDegrees=30, twistLimitDegrees=5}
 {bone1Name=dyn_cape_r_04, bone2Name=dyn_cape_r_03, swing1LimitDegrees=35, swing2LimitDegrees=45, twistLimitDegrees=5}
 {bone1Name=dyn_cape_r_05, bone2Name=dyn_cape_r_04, swing1LimitDegrees=35, swing2LimitDegrees=45, twistLimitDegrees=5}
 {bone1Name=dyn_cape_r_06, bone2Name=dyn_cape_r_05, swing1LimitDegrees=35, swing2LimitDegrees=45, twistLimitDegrees=5}
 {
  bone1Name=dyn_cape_pin_l_01
  bone2Name=clavicle_l
  swing1LimitDegrees=45
  swing1Motion=Free
  swing2LimitDegrees=45
  swing2Motion=Free
  twistLimitDegrees=45
  twistMotion=Free
 }
 {bone1Name=dyn_cape_l_01, bone2Name=dyn_cape_pin_l_01, swing1LimitDegrees=10, swing2LimitDegrees=10, twistLimitDegrees=5}
 {bone1Name=dyn_cape_l_02, bone2Name=dyn_cape_l_01, swing1LimitDegrees=25, swing2LimitDegrees=25, twistLimitDegrees=5}
 {bone1Name=dyn_cape_l_03, bone2Name=dyn_cape_l_02, swing1LimitDegrees=35, swing2LimitDegrees=30, twistLimitDegrees=5}
 {bone1Name=dyn_cape_l_04, bone2Name=dyn_cape_l_03, swing1LimitDegrees=35, swing2LimitDegrees=45, twistLimitDegrees=5}
 {bone1Name=dyn_cape_l_05, bone2Name=dyn_cape_l_04, swing1LimitDegrees=35, swing2LimitDegrees=45, twistLimitDegrees=5}
 {bone1Name=dyn_cape_l_06, bone2Name=dyn_cape_l_05, swing1LimitDegrees=35, swing2LimitDegrees=45, twistLimitDegrees=5}
 {
  bone1Name=thigh_l
  bone2Name=pelvis
  swing1LimitDegrees=45
  swing1Motion=Free
  swing2LimitDegrees=45
  swing2Motion=Free
  twistLimitDegrees=45
  twistMotion=Free
 }
 {
  bone1Name=calf_l
  bone2Name=thigh_l
  swing1LimitDegrees=45
  swing1Motion=Free
  swing2LimitDegrees=45
  swing2Motion=Free
  twistLimitDegrees=45
  twistMotion=Free
 }
 {
  bone1Name=calf_r
  bone2Name=thigh_r
  swing1LimitDegrees=45
  swing1Motion=Free
  swing2LimitDegrees=45
  swing2Motion=Free
  twistLimitDegrees=45
  twistMotion=Free
 }
 {
  bone1Name=thigh_r
  bone2Name=pelvis
  swing1LimitDegrees=45
  swing1Motion=Free
  swing2LimitDegrees=45
  swing2Motion=Free
  twistLimitDegrees=45
  twistMotion=Free
 }
 {
  bone1Name=clavicle_r
  bone2Name=spine_05
  swing1LimitDegrees=45
  swing1Motion=Free
  swing2LimitDegrees=45
  swing2Motion=Free
  twistLimitDegrees=45
  twistMotion=Free
 }
 {
  bone1Name=lowerarm_r
  bone2Name=upperarm_r
  swing1LimitDegrees=45
  swing1Motion=Free
  swing2LimitDegrees=45
  swing2Motion=Free
  twistLimitDegrees=45
  twistMotion=Free
 }
 {
  bone1Name=dyn_feather_l_MSTR
  bone2Name=clavicle_l
  swing1LimitDegrees=45
  swing1Motion=Free
  swing2LimitDegrees=45
  swing2Motion=Free
  twistLimitDegrees=45
  twistMotion=Free
 }
 {bone1Name=dyn_feather_l, bone2Name=dyn_feather_l_MSTR, swing1LimitDegrees=10, swing2LimitDegrees=15, twistLimitDegrees=5}
 {
  bone1Name=dyn_feather_bk_l_MSTR
  bone2Name=clavicle_l
  swing1LimitDegrees=45
  swing1Motion=Free
  swing2LimitDegrees=45
  swing2Motion=Free
  twistLimitDegrees=45
  twistMotion=Free
 }
 {bone1Name=dyn_feather_bk_l, bone2Name=dyn_feather_bk_l_MSTR, swing1LimitDegrees=10, swing2LimitDegrees=15, twistLimitDegrees=5}
 {
  bone1Name=dyn_feather_bk_r_MSTR
  bone2Name=clavicle_r
  swing1LimitDegrees=45
  swing1Motion=Free
  swing2LimitDegrees=45
  swing2Motion=Free
  twistLimitDegrees=45
  twistMotion=Free
 }
 {bone1Name=dyn_feather_bk_r, bone2Name=dyn_feather_bk_r_MSTR, swing1LimitDegrees=10, swing2LimitDegrees=15, twistLimitDegrees=5}
 {
  bone1Name=dyn_feather_r_MSTR
  bone2Name=clavicle_r
  swing1LimitDegrees=45
  swing1Motion=Free
  swing2LimitDegrees=45
  swing2Motion=Free
  twistLimitDegrees=45
  twistMotion=Free
 }
 {bone1Name=dyn_feather_r, bone2Name=dyn_feather_r_MSTR, swing1LimitDegrees=10, swing2LimitDegrees=15, twistLimitDegrees=5}
 {
  bone1Name=dyn_feather_fr_l_MSTR
  bone2Name=spine_05
  swing1LimitDegrees=45
  swing1Motion=Free
  swing2LimitDegrees=45
  swing2Motion=Free
  twistLimitDegrees=45
  twistMotion=Free
 }
 {bone1Name=dyn_feather_fr_l, bone2Name=dyn_feather_fr_l_MSTR, swing1LimitDegrees=10, swing2LimitDegrees=15, twistLimitDegrees=5}
 {
  bone1Name=dyn_feather_fr_r_MSTR
  bone2Name=spine_05
  swing1LimitDegrees=45
  swing1Motion=Free
  swing2LimitDegrees=45
  swing2Motion=Free
  twistLimitDegrees=45
  twistMotion=Free
 }
 {bone1Name=dyn_feather_fr_r, bone2Name=dyn_feather_fr_r_MSTR, swing1LimitDegrees=10, swing2LimitDegrees=15, twistLimitDegrees=5}
 {
  bone1Name=upperarm_r
  bone2Name=clavicle_r
  swing1LimitDegrees=45
  swing1Motion=Free
  swing2LimitDegrees=45
  swing2Motion=Free
  twistLimitDegrees=45
  twistMotion=Free
 }
 {
  bone1Name=dyn_cape_pin_mid
  bone2Name=spine_05
  swing1LimitDegrees=45
  swing1Motion=Free
  swing2LimitDegrees=45
  swing2Motion=Free
  twistLimitDegrees=45
  twistMotion=Free
 }
 {bone1Name=dyn_cape_mid_01, bone2Name=dyn_cape_pin_mid, swing1LimitDegrees=15, swing2LimitDegrees=10, twistLimitDegrees=5}
 {bone1Name=dyn_cape_mid_02, bone2Name=dyn_cape_mid_01, swing1LimitDegrees=25, swing2LimitDegrees=25, twistLimitDegrees=5}
 {bone1Name=dyn_cape_mid_03, bone2Name=dyn_cape_mid_02, swing1LimitDegrees=35, swing2LimitDegrees=30, twistLimitDegrees=5}
 {bone1Name=dyn_cape_mid_04, bone2Name=dyn_cape_mid_03, swing1LimitDegrees=35, swing2LimitDegrees=45, twistLimitDegrees=5}
 {bone1Name=dyn_cape_mid_05, bone2Name=dyn_cape_mid_04, swing1LimitDegrees=35, swing2LimitDegrees=45, twistLimitDegrees=5}
 {bone1Name=dyn_cape_mid_06, bone2Name=dyn_cape_mid_05, swing1LimitDegrees=35, swing2LimitDegrees=45, twistLimitDegrees=5}
 {
  bone1Name=dyn_cape_r_03
  bone2Name=dyn_cape_mid_03
  swing1LimitDegrees=45
  swing1Motion=Free
  swing2LimitDegrees=45
  swing2Motion=Free
  twistLimitDegrees=45
  twistMotion=Free
 }
 {
  bone1Name=dyn_cape_l_03
  bone2Name=dyn_cape_mid_03
  swing1LimitDegrees=45
  swing1Motion=Free
  swing2LimitDegrees=45
  swing2Motion=Free
  twistLimitDegrees=45
  twistMotion=Free
 }
 {
  bone1Name=dyn_cape_r_04
  bone2Name=dyn_cape_mid_04
  swing1LimitDegrees=45
  swing1Motion=Free
  swing2LimitDegrees=45
  swing2Motion=Free
  twistLimitDegrees=45
  twistMotion=Free
 }
 {
  bone1Name=dyn_cape_l_04
  bone2Name=dyn_cape_mid_04
  swing1LimitDegrees=45
  swing1Motion=Free
  swing2LimitDegrees=45
  swing2Motion=Free
  twistLimitDegrees=45
  twistMotion=Free
 }
 {
  bone1Name=dyn_cape_r_01
  bone2Name=dyn_cape_mid_01
  swing1LimitDegrees=45
  swing1Motion=Free
  swing2LimitDegrees=45
  swing2Motion=Free
  twistLimitDegrees=45
  twistMotion=Free
 }
 {
  bone1Name=dyn_cape_l_01
  bone2Name=dyn_cape_mid_01
  swing1LimitDegrees=45
  swing1Motion=Free
  swing2LimitDegrees=45
  swing2Motion=Free
  twistLimitDegrees=45
  twistMotion=Free
 }
 {
  bone1Name=dyn_cape_r_01
  bone2Name=dyn_cape_l_01
  swing1LimitDegrees=45
  swing1Motion=Free
  swing2LimitDegrees=45
  swing2Motion=Free
  twistLimitDegrees=45
  twistMotion=Free
 }
]
