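# Physics asset for the F_MED_Obsidian character's dynamic cape and feathers, a variant named for the Sunburst emote.
# Capsule bodies cover upper spine, clavicles, four clavicle-attached feathers and three six-bone cape chains hung from left clavicle, spine_05 and right clavicle.
# Chain constraints allow little twist, with swing limits widening down from the root.
# Angularly free cross-constraints link side chains to the middle at segments 1, 3 and 4, and to each other at the root.
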PhysicsAsset /BRCosmetics/Characters/Player/Female/Medium/Bodies/F_MED_Obsidian/Meshes/F_MED_Obsidian_Dyn_cape_SunburstEmote
bodies {
 clavicle_l [{shapeType=Capsule, length=6.125, radius=8.005, center=<6.4728 0.344 -3.7639>, rotation=<rot 83.0162 10.3785 6.1841>}]
 clavicle_r [{shapeType=Capsule, length=6.125, radius=8.005, center=<-6.4728 -0.344 3.7639>, rotation=<rot 83.0161 10.3785 -173.8159>}]
 dyn_cape_l_01 [{shapeType=Capsule, length=13.2812, radius=15, center=<9.6403 0 15>, rotation=<rot -90 -0.0033 0.0033>}]
 dyn_cape_l_02 [{shapeType=Capsule, length=13.2812, radius=15, center=<9.6403 0 15>, rotation=<rot -90 0.0396 -0.0396>}]
 dyn_cape_l_03 [{shapeType=Capsule, length=13.2812, radius=15, center=<9.6403 0 15>, rotation=<rot -90 0.0396 -0.0396>}]
 dyn_cape_l_04 [{shapeType=Capsule, length=13.2812, radius=15, center=<9.6403 0 15>, rotation=<rot -90 0.0396 -0.0396>}]
 dyn_cape_l_05 [{shapeType=Capsule, length=13.2812, radius=15, center=<9.6403 0 15>, rotation=<rot -90 0.0396 -0.0396>}]
 dyn_cape_l_06 [{shapeType=Capsule, length=7.2188, radius=15, center=<7.0061 -0.0002 15>, rotation=<rot -90 0.0099 -0.0099>}]
 dyn_cape_mid_01 [{shapeType=Capsule, length=13.2812, radius=15, center=<9.6403 0 15>, rotation=<rot -90 0.0396 90>}]
 dyn_cape_mid_02 [{shapeType=Capsule, length=13.2812, radius=15, center=<9.6403 0 15>, rotation=<rot -90 0.0396 90>}]
 dyn_cape_mid_03 [{shapeType=Capsule, length=13.2812, radius=15, center=<9.6403 0 15>, rotation=<rot -90 0.0396 90>}]
 dyn_cape_mid_04 [{shapeType=Capsule, length=13.2812, radius=15, center=<9.6403 0 15>, rotation=<rot -90 0.0396 90>}]
 dyn_cape_mid_05 [{shapeType=Capsule, length=13.2812, radius=15, center=<9.6403 0 15>, rotation=<rot -90 0.0396 90>}]
 dyn_cape_mid_06 [{shapeType=Capsule, length=15.4688, radius=15, center=<7.0061 -0.0002 15>, rotation=<rot -90 0.0099 90>}]
 dyn_cape_r_01 [{shapeType=Capsule, length=13.2812, radius=15, center=<9.6403 0 15>, rotation=<rot -90 0.0396 -0.0396>}]
 dyn_cape_r_02 [{shapeType=Capsule, length=13.2812, radius=15, center=<9.6403 0 15>, rotation=<rot -90 0.0396 -0.0396>}]
 dyn_cape_r_03 [{shapeType=Capsule, length=13.2812, radius=15, center=<9.6403 0 15>, rotation=<rot -90 0.0396 -0.0396>}]
 dyn_cape_r_04 [{shapeType=Capsule, length=13.2812, radius=15, center=<9.6403 0 15>, rotation=<rot -90 0.0396 -0.0396>}]
 dyn_cape_r_05 [{shapeType=Capsule, length=13.2812, radius=15, center=<9.6403 0 15>, rotation=<rot -90 0.0396 -0.0396>}]
 dyn_cape_r_06 [{shapeType=Capsule, length=7.2188, radius=15, center=<7.0061 -0.0002 15>, rotation=<rot -90 0.0099 -0.0099>}]
 dyn_feather_bk_l [{shapeType=Capsule, length=15.6875, radius=7.9688, center=<5.7886 0.0003 0>, rotation=<rot -90 0.0099 -0.0099>}]
 dyn_feather_bk_r [{shapeType=Capsule, length=15.6875, radius=7.9688, center=<5.7886 0.0003 0>, rotation=<rot -90 0.0099 -0.0099>}]
 dyn_feather_l [{shapeType=Capsule, length=15.6875, radius=7.9688, center=<5.7886 0.0003 0>, rotation=<rot -90 0.0099 -0.0099>}]
 dyn_feather_r [{shapeType=Capsule, length=15.6875, radius=7.9688, center=<5.7886 0.0003 0>, rotation=<rot -90 0.0099 -0.0099>}]
 spine_05 [{shapeType=Capsule, length=14.2188, radius=11.1613, center=<-2.6173 2.0295 -0.0006>}]
}
constraints [
 {
  bone1Name=clavicle_l
  bone2Name=spine_05
  swing1LimitDegrees=45
  swing1Motion=Free
  swing2LimitDegrees=45
  swing2Motion=Free
  twistLimitDegrees=45
  twistMotion=Free
 }
 {bone1Name=dyn_cape_r_01, bone2Name=clavicle_r, swing1LimitDegrees=10, swing2LimitDegrees=10, twistLimitDegrees=5}
 {bone1Name=dyn_cape_r_02, bone2Name=dyn_cape_r_01, swing1LimitDegrees=25, swing2LimitDegrees=25, twistLimitDegrees=5}
 {bone1Name=dyn_cape_r_03, bone2Name=dyn_cape_r_02, swing1LimitDegrees=35, swing2LimitDegrees=30, twistLimitDegrees=5}
 {bone1Name=dyn_cape_r_04, bone2Name=dyn_cape_r_03, swing1LimitDegrees=35, swing2LimitDegrees=45, twistLimitDegrees=5}
 {bone1Name=dyn_cape_r_05, bone2Name=dyn_cape_r_04, swing1LimitDegrees=35, swing2LimitDegrees=45, twistLimitDegrees=5}
 {bone1Name=dyn_cape_r_06, bone2Name=dyn_cape_r_05, swing1LimitDegrees=35, swing2LimitDegrees=45, twistLimitDegrees=5}
 {bone1Name=dyn_cape_l_01, bone2Name=clavicle_l, swing1LimitDegrees=10, swing2LimitDegrees=10, twistLimitDegrees=5}
 {bone1Name=dyn_cape_l_02, bone2Name=dyn_cape_l_01, swing1LimitDegrees=25, swing2LimitDegrees=25, twistLimitDegrees=5}
 {bone1Name=dyn_cape_l_03, bone2Name=dyn_cape_l_02, swing1LimitDegrees=35, swing2LimitDegrees=30, twistLimitDegrees=5}
 {bone1Name=dyn_cape_l_04, bone2Name=dyn_cape_l_03, swing1LimitDegrees=35, swing2LimitDegrees=45, twistLimitDegrees=5}
 {bone1Name=dyn_cape_l_05, bone2Name=dyn_cape_l_04, swing1LimitDegrees=35, swing2LimitDegrees=45, twistLimitDegrees=5}
 {bone1Name=dyn_cape_l_06, bone2Name=dyn_cape_l_05, swing1LimitDegrees=35, swing2LimitDegrees=45, twistLimitDegrees=5}
 {
  bone1Name=clavicle_r
  bone2Name=spine_05
  swing1LimitDegrees=45
  swing1Motion=Free
  swing2LimitDegrees=45
  swing2Motion=Free
  twistLimitDegrees=45
  twistMotion=Free
 }
 {bone1Name=dyn_feather_l, bone2Name=clavicle_l, swing1LimitDegrees=10, swing2LimitDegrees=15, twistLimitDegrees=5}
 {bone1Name=dyn_feather_bk_l, bone2Name=clavicle_l, swing1LimitDegrees=10, swing2LimitDegrees=15, twistLimitDegrees=5}
 {bone1Name=dyn_feather_bk_r, bone2Name=clavicle_r, swing1LimitDegrees=10, swing2LimitDegrees=15, twistLimitDegrees=5}
 {bone1Name=dyn_feather_r, bone2Name=clavicle_r, swing1LimitDegrees=10, swing2LimitDegrees=15, twistLimitDegrees=5}
 {bone1Name=dyn_cape_mid_01, bone2Name=spine_05, swing1LimitDegrees=15, swing2LimitDegrees=10, twistLimitDegrees=5}
 {bone1Name=dyn_cape_mid_02, bone2Name=dyn_cape_mid_01, swing1LimitDegrees=25, swing2LimitDegrees=25, twistLimitDegrees=5}
 {bone1Name=dyn_cape_mid_03, bone2Name=dyn_cape_mid_02, swing1LimitDegrees=35, swing2LimitDegrees=30, twistLimitDegrees=5}
 {bone1Name=dyn_cape_mid_04, bone2Name=dyn_cape_mid_03, swing1LimitDegrees=35, swing2LimitDegrees=45, twistLimitDegrees=5}
 {bone1Name=dyn_cape_mid_05, bone2Name=dyn_cape_mid_04, swing1LimitDegrees=35, swing2LimitDegrees=45, twistLimitDegrees=5}
 {bone1Name=dyn_cape_mid_06, bone2Name=dyn_cape_mid_05, swing1LimitDegrees=35, swing2LimitDegrees=45, twistLimitDegrees=5}
 {
  bone1Name=dyn_cape_r_03
  bone2Name=dyn_cape_mid_03
  swing1LimitDegrees=45
  swing1Motion=Free
  swing2LimitDegrees=45
  swing2Motion=Free
  twistLimitDegrees=45
  twistMotion=Free
 }
 {
  bone1Name=dyn_cape_l_03
  bone2Name=dyn_cape_mid_03
  swing1LimitDegrees=45
  swing1Motion=Free
  swing2LimitDegrees=45
  swing2Motion=Free
  twistLimitDegrees=45
  twistMotion=Free
 }
 {
  bone1Name=dyn_cape_r_04
  bone2Name=dyn_cape_mid_04
  swing1LimitDegrees=45
  swing1Motion=Free
  swing2LimitDegrees=45
  swing2Motion=Free
  twistLimitDegrees=45
  twistMotion=Free
 }
 {
  bone1Name=dyn_cape_l_04
  bone2Name=dyn_cape_mid_04
  swing1LimitDegrees=45
  swing1Motion=Free
  swing2LimitDegrees=45
  swing2Motion=Free
  twistLimitDegrees=45
  twistMotion=Free
 }
 {
  bone1Name=dyn_cape_r_01
  bone2Name=dyn_cape_mid_01
  swing1LimitDegrees=45
  swing1Motion=Free
  swing2LimitDegrees=45
  swing2Motion=Free
  twistLimitDegrees=45
  twistMotion=Free
 }
 {
  bone1Name=dyn_cape_l_01
  bone2Name=dyn_cape_mid_01
  swing1LimitDegrees=45
  swing1Motion=Free
  swing2LimitDegrees=45
  swing2Motion=Free
  twistLimitDegrees=45
  twistMotion=Free
 }
 {
  bone1Name=dyn_cape_r_01
  bone2Name=dyn_cape_l_01
  swing1LimitDegrees=45
  swing1Motion=Free
  swing2LimitDegrees=45
  swing2Motion=Free
  twistLimitDegrees=45
  twistMotion=Free
 }
]
