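# Physics asset for the dynamic loincloth on the F_MED_Obsidian female player body.
# The loincloth is a four-bone chain of three capsules and an end sphere, hung from the pelvis through a small master body, with narrow twist and swing limits that widen toward the tip.
# Pelvis, thigh and calf capsules, linked by free angular constraints, give it body shapes, alongside an unconstrained wide, flat box on dyn_root_slope.
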PhysicsAsset /BRCosmetics/Characters/Player/Female/Medium/Bodies/F_MED_Obsidian/Meshes/F_MED_Obsidian_Dyn_loincloth
bodies {
 calf_l [{shapeType=Capsule, length=35.25, radius=9.0988, center=<-21.1406 -5.487 -1.6507>, rotation=<rot 90 1.7899 -3.5629>}]
 calf_r [{shapeType=Capsule, length=35.25, radius=9.0988, center=<21.1406 5.487 1.6507>, rotation=<rot 90 -178.2101 -3.5629>}]
 dyn_loin_01 [{shapeType=Capsule, length=4.25, radius=7.5, shapeName=dyn_roin_r_01_capsule, center=<8 0 4>, rotation=<rot 90 0 0>}]
 dyn_loin_02 [{shapeType=Capsule, length=4.75, radius=7.5, shapeName=dyn_roin_r_02_capsule, center=<8 0 4>, rotation=<rot 90 0 0>}]
 dyn_loin_03 [{shapeType=Capsule, length=3.25, radius=7.75, shapeName=dyn_roin_r_03_capsule, center=<8 0 4>, rotation=<rot 90 0 0>}]
 dyn_loin_04 [{shapeType=Sphere, radius=11, shapeName=dyn_loin_04_sphere, center=<12 0 7>}]
 dyn_loin_MSTR [{shapeType=Capsule, length=0.5, radius=0.5}]
 dyn_root_slope [{shapeType=Box, extentX=231.0625, extentY=231.0625, extentZ=12.375, center=<0 0 -5.9184>}]
 pelvis [{shapeType=Capsule, length=37.9375, radius=25.5625, center=<-23.967 -28.9653 0.0001>, rotation=<rot -90 0 25.0025>}]
 thigh_l [{shapeType=Capsule, length=30.875, radius=13.38, center=<-23.4022 -4.7188 -5.6095>, rotation=<rot 89.3173 -166.2119 -168.8126>}]
 thigh_r [{shapeType=Capsule, length=30.875, radius=13.38, center=<23.4022 4.7189 5.6093>, rotation=<rot 89.3233 -166.214 11.1811>}]
}
constraints [
 {
  bone1Name=thigh_l
  bone2Name=pelvis
  swing1LimitDegrees=45
  swing1Motion=Free
  swing2LimitDegrees=45
  swing2Motion=Free
  twistLimitDegrees=45
  twistMotion=Free
 }
 {
  bone1Name=calf_l
  bone2Name=thigh_l
  swing1LimitDegrees=45
  swing1Motion=Free
  swing2LimitDegrees=45
  swing2Motion=Free
  twistLimitDegrees=45
  twistMotion=Free
 }
 {
  bone1Name=calf_r
  bone2Name=thigh_r
  swing1LimitDegrees=45
  swing1Motion=Free
  swing2LimitDegrees=45
  swing2Motion=Free
  twistLimitDegrees=45
  twistMotion=Free
 }
 {
  bone1Name=thigh_r
  bone2Name=pelvis
  swing1LimitDegrees=45
  swing1Motion=Free
  swing2LimitDegrees=45
  swing2Motion=Free
  twistLimitDegrees=45
  twistMotion=Free
 }
 {
  bone1Name=dyn_loin_MSTR
  bone2Name=pelvis
  swing1LimitDegrees=45
  swing1Motion=Free
  swing2LimitDegrees=45
  swing2Motion=Free
  twistLimitDegrees=45
  twistMotion=Free
 }
 {bone1Name=dyn_loin_01, bone2Name=dyn_loin_MSTR, swing1LimitDegrees=5, swing2LimitDegrees=2, twistLimitDegrees=2}
 {bone1Name=dyn_loin_02, bone2Name=dyn_loin_01, swing1LimitDegrees=5, swing2LimitDegrees=10, twistLimitDegrees=2}
 {bone1Name=dyn_loin_03, bone2Name=dyn_loin_02, swing1LimitDegrees=10, swing2LimitDegrees=25, twistLimitDegrees=2}
 {bone1Name=dyn_loin_04, bone2Name=dyn_loin_03, swing1LimitDegrees=10, swing2LimitDegrees=25, twistLimitDegrees=2}
]
