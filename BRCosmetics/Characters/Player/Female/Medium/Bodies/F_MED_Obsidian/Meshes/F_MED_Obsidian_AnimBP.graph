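# Animation Blueprint for the F_MED_Obsidian female body cosmetic, adding secondary motion to its cape, loincloth and belt discs.
# It copies the attached parent's pose, runs two deform Control Rigs, then applies weighted constraints and bone offsets driven by movement state.
# Cached branches simulate the loincloth with RigidBody, including a Chugga variant, or trail the cape.
# Output selects the rigid-body branch when enabled, otherwise a DBNO or standard trail branch.
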AnimBlueprint /BRCosmetics/Characters/Player/Female/Medium/Bodies/F_MED_Obsidian/Meshes/F_MED_Obsidian_AnimBP
target_skeleton=&/F_MED_Obsidian_Skeleton
graphs [
 AnimGraph {nodes [
   n1 = "Constraint - dyn_loin_l_MSTR (PPPPPPP)"!(
    ComponentPose
    Alpha=1
    ConstraintWeights_0=0.2933
    ConstraintWeights_1:float=0.0001
    ConstraintWeights_2:float=0.4344
    ConstraintWeights_3:float=0.0738
    ConstraintWeights_4:float=0.1984
    ConstraintWeights_5:float=0
    ConstraintWeights_6:float=0
   ) -> (Pose) {
    alphaBoolBlend {blendOption=Linear}
    boneToModify {boneName=dyn_loin_l_MSTR}
    constraintSetup [
     {targetBone {boneName=deform_thigh_fr_l}}
     {targetBone {boneName=thigh_twist_01_r}}
     {targetBone {boneName=thigh_twist_01_l}}
     {targetBone {boneName=thigh_l}}
     {targetBone {boneName=pelvis}}
     {targetBone {boneName=deform_thigh_bk_l}}
     {targetBone {boneName=deform_thigh_fr_r}}
    ]
    constraintWeights [1,1,1,1,1,1,1]
   }
   "Constraint - dyn_roin_r_MSTR (PPPPPPP)"!(
    ComponentPose=@n1
    Alpha=1
    ConstraintWeights_0=0
    ConstraintWeights_1:float=0
    ConstraintWeights_2:float=0.0738
    ConstraintWeights_3:float=0.4344
    ConstraintWeights_4:float=0.0001
    ConstraintWeights_5:float=0.1984
    ConstraintWeights_6:float=0.2933
   ) -> (Pose) {
    alphaBoolBlend {blendOption=Linear}
    boneToModify {boneName=dyn_roin_r_MSTR}
    constraintSetup [
     {targetBone {boneName=deform_thigh_fr_l}}
     {targetBone {boneName=deform_thigh_bk_r}}
     {targetBone {boneName=thigh_r}}
     {targetBone {boneName=thigh_twist_01_r}}
     {targetBone {boneName=thigh_twist_01_l}}
     {targetBone {boneName=pelvis}}
     {targetBone {boneName=deform_thigh_fr_r}}
    ]
    constraintWeights [1,1,1,1,1,1,1]
   }
   n2 = "Transform (Modify) Bone - Bone: dyn_cape_r_01"(Alpha=1, Translation="0,0,0", Rotation="5,0,0", Scale="1,1,1") {
    rotationMode=BMM_Additive
    rotationSpace=BCS_BoneSpace
    alphaBoolBlend {blendOption=Linear}
    boneToModify {boneName=dyn_cape_r_01}
    rotation {pitch=5}
   }
   n3 = "Transform (Modify) Bone - Bone: dyn_cape_mid_r_02"(ComponentPose=@n2, Alpha=1, Translation="0,0,0", Rotation="-10,0,0", Scale="1,1,1") {
    rotationMode=BMM_Additive
    rotationSpace=BCS_BoneSpace
    alphaBoolBlend {blendOption=Linear}
    boneToModify {boneName=dyn_cape_mid_r_02}
    rotation {pitch=-10}
   }
   "Transform (Modify) Bone - Bone: dyn_cape_mid_r_05"(ComponentPose=@n3, Alpha=1, Translation="0,0,0", Rotation="-25,0,0", Scale="1,1,1") {
    rotationMode=BMM_Additive
    rotationSpace=BCS_BoneSpace
    alphaBoolBlend {blendOption=Linear}
    boneToModify {boneName=dyn_cape_mid_r_05}
    rotation {pitch=-25}
   }
   n4 = Get crouch alpha() -> (crouch alpha:double)
   n5 = "Transform (Modify) Bone - Bone: dyn_belt_discs_fr_fr_l"(Alpha=@n4, Translation="-2,-2,1", Rotation="10,-30,0", Scale="1,1,1") {
    rotationMode=BMM_Additive
    rotationSpace=BCS_BoneSpace
    translationMode=BMM_Additive
    translationSpace=BCS_BoneSpace
    alphaBoolBlend {blendOption=Linear}
    boneToModify {boneName=dyn_belt_discs_fr_fr_l}
    rotation {pitch=10, yaw=-30}
    translation=<-2 -2 1>
   }
   "Transform (Modify) Bone - Bone: dyn_belt_discs_fr"(ComponentPose=@n5, Alpha=@n4, Translation="-2,-1,1", Rotation="10,-30,0", Scale="1,1,1") {
    rotationMode=BMM_Additive
    rotationSpace=BCS_BoneSpace
    translationMode=BMM_Additive
    translationSpace=BCS_BoneSpace
    alphaBoolBlend {blendOption=Linear}
    boneToModify {boneName=dyn_belt_discs_fr}
    rotation {pitch=10, yaw=-30}
    translation=<-2 -1 1>
   }
   RigidBody!(ComponentPose, Alpha, OverrideWorldGravity:Vector=Get GravityOverrideLoincloth() -> (GravityOverrideLoincloth:Vector), ExternalForce) -> (Pose) {
    bClampLinearTranslationLimitToRefPose=true
    bOverrideWorldGravity=true
    overridePhysicsAsset=&/F_MED_Obsidian_Dyn_loincloth
    simulationSpace=BaseBoneSpace
    alphaBoolBlend {blendOption=Linear}
    baseBoneRef {boneName=dyn_root_simspace}
    componentAppliedLinearAccClamp=<3000 3000 3000>
    componentLinearAccScale {z=0.5}
    componentLinearVelScale {z=0.5}
    overrideWorldGravity {z=-980}
    simSpaceSettings {
     maxAngularAcceleration=1000
     maxAngularVelocity=3
     maxLinearAcceleration=700
     maxLinearVelocity=500
     velocityScaleZ=0
     worldAlpha=1
     externalLinearDragV=<1 1>
    }
   }
   n6 = Use cached pose 'trailCache'()
   n7 = Use cached pose 'dbnoCache'()
   n8 = Get bIsDBNO()
   Output Pose(Result=Blend Poses by bool(
     BlendPose_0=Use cached pose 'rigidCache'()
     BlendPose_1=Blend Poses by bool(BlendPose_0=@n7, BlendPose_1=@n6, bActiveValue=@n8) {blendType=Linear}
     bActiveValue=Get bRigidBodyEnabled()
    ) {blendType=Linear}) {layerGroup=None}
   n9 = RigidBody(ComponentPose=Local To Component(LocalPose=Use cached pose 'chuggaEmote'()), OverrideWorldGravity="0,0,0") {
    bClampLinearTranslationLimitToRefPose=true
    bOverrideWorldGravity=true
    overridePhysicsAsset=&/F_MED_Obsidian_Dyn_lnclth_Chugga
    simulationSpace=BaseBoneSpace
    alphaBoolBlend {blendOption=Linear}
    baseBoneRef {boneName=dyn_root_simspace}
    componentAppliedLinearAccClamp=<3000 3000 3000>
    componentLinearAccScale {z=0.5}
    componentLinearVelScale {z=0.5}
    overrideWorldGravity {z=-980}
    simSpaceSettings {
     maxAngularAcceleration=1000
     maxAngularVelocity=3
     maxLinearAcceleration=700
     maxLinearVelocity=500
     velocityScaleZ=0
     worldAlpha=1
     externalLinearDragV=<1 1>
    }
   }
   n10 = "Transform (Modify) Bone - Bone: dyn_loin_MSTR"(ComponentPose=@n9, Alpha=1, Translation="0,0,0", Rotation="-90,0,0", Scale="1,1,1") {
    rotationMode=BMM_Additive
    rotationSpace=BCS_BoneSpace
    alphaBoolBlend {blendOption=Linear}
    boneToModify {boneName=dyn_loin_MSTR}
    rotation {pitch=-90}
   }
   n11 = Component To Local(ComponentPose=@n10)
   n12 = RigidBody(ComponentPose=Local To Component(LocalPose=Use cached pose 'chuggaEmote'()), OverrideWorldGravity=Get GravityOverrideLoincloth()) {
    bClampLinearTranslationLimitToRefPose=true
    bOverrideWorldGravity=true
    overridePhysicsAsset=&F_MED_Obsidian_Dyn_loincloth
    simulationSpace=BaseBoneSpace
    alphaBoolBlend {blendOption=Linear}
    baseBoneRef {boneName=dyn_root_simspace}
    componentAppliedLinearAccClamp=<3000 3000 3000>
    componentLinearAccScale {z=0.5}
    componentLinearVelScale {z=0.5}
    overrideWorldGravity {z=-980}
    simSpaceSettings {
     maxAngularAcceleration=1000
     maxAngularVelocity=3
     maxLinearAcceleration=700
     maxLinearVelocity=500
     velocityScaleZ=0
     worldAlpha=1
     externalLinearDragV=<1 1>
    }
   }
   n13 = Get Is_BaseballSliding() -> (Is_BaseballSliding:double)
   n14 = "Transform (Modify) Bone - Bone: dyn_loin_01"(ComponentPose=@n12, Alpha=@n13, Translation="0,0,0", Rotation="-50.000,-50.000,0.000", Scale="1,1,1") {
    rotationMode=BMM_Additive
    rotationSpace=BCS_BoneSpace
    translationMode=BMM_Additive
    translationSpace=BCS_BoneSpace
    alphaBoolBlend {blendInTime=0.5, blendOption=Linear, blendOutTime=0.2}
    alphaScaleBiasClamp {bInterpResult=true, interpSpeedDecreasing=5, interpSpeedIncreasing=3}
    boneToModify {boneName=dyn_loin_01}
    rotation=<rot -49.9999 -50 0.0001>
   }
   n15 = "Transform (Modify) Bone - Bone: dyn_loin_03"(ComponentPose=@n14, Alpha=@n13, Translation="0,0,0", Rotation="-10.000,-60.000,0.000", Scale="1,1,1") {
    rotationMode=BMM_Additive
    rotationSpace=BCS_BoneSpace
    translationMode=BMM_Additive
    translationSpace=BCS_BoneSpace
    alphaBoolBlend {blendInTime=0.5, blendOption=Linear, blendOutTime=0.2}
    alphaScaleBiasClamp {bInterpResult=true, interpSpeedDecreasing=5, interpSpeedIncreasing=5}
    boneToModify {boneName=dyn_loin_03}
    rotation=<rot -9.9999 -60 0>
   }
   n16 = "Transform (Modify) Bone - Bone: dyn_loin_03"!(ComponentPose=@n15, bAlphaBoolEnabled:bool=True, Translation="-0.000,0.144,2.220", Rotation="0,0,0", Scale="1,1,1") -> (Pose) {
    alphaInputType=Bool
    rotationMode=BMM_Additive
    rotationSpace=BCS_BoneSpace
    translationMode=BMM_Additive
    translationSpace=BCS_BoneSpace
    alphaBoolBlend {blendInTime=0.5, blendOption=Linear, blendOutTime=0.2}
    alphaScaleBiasClamp {bInterpResult=true, interpSpeedDecreasing=5, interpSpeedIncreasing=5}
    boneToModify {boneName=dyn_loin_03}
    rotation=<rot 0.0001 0 0>
    translation {y=0.144, z=2.22}
   }
   n17 = Component To Local(ComponentPose=@n16)
   n18 = "Transform (Modify) Bone - Bone: dyn_loin_MSTR"!(
    ComponentPose=Local To Component(LocalPose=Blend Poses by bool(BlendPose_0=@n11, BlendPose_1=@n17) {blendType=Linear})
    bAlphaBoolEnabled:bool=Get is_crouch_moving() -> (is_crouch_moving:bool)
    Translation="0,0,-5"
    Rotation="0,0,0"
    Scale="1,1,1"
   ) -> (Pose) {
    alphaInputType=Bool
    rotationMode=BMM_Additive
    rotationSpace=BCS_BoneSpace
    translationMode=BMM_Additive
    translationSpace=BCS_BoneSpace
    alphaBoolBlend {blendInTime=0.2, blendOption=Linear, blendOutTime=0.2}
    boneToModify {boneName=dyn_loin_MSTR}
    translation {z=-5}
   }
   Save cached pose 'rigidCache'(Pose=Component To Local(ComponentPose=@n18)) {cacheName=rigidCache}
   n19 = Get TrailControllerRot() -> (TrailControllerRot:Rotator)
   n20 = "Transform (Modify) Bone - Bone: dyn_cape_l_01"(ComponentPose=Local To Component(LocalPose=Use cached pose 'Start_Cache'()), Alpha=1, Translation="0,0,0", Rotation=@n19, Scale="1,1,1") {rotationMode=BMM_Additive, rotationSpace=BCS_BoneSpace, alphaBoolBlend {blendOption=Linear}, boneToModify {boneName=dyn_cape_l_01}}
   n21 = "Transform (Modify) Bone - Bone: dyn_cape_r_01"(ComponentPose=@n20, Alpha=1, Translation="0,0,0", Rotation=@n19, Scale="1,1,1") {rotationMode=BMM_Additive, rotationSpace=BCS_BoneSpace, alphaBoolBlend {blendOption=Linear}, boneToModify {boneName=dyn_cape_r_01}}
   n22 = "Transform (Modify) Bone - Bone: dyn_cape_mid_01"(ComponentPose=@n21, Alpha=1, Translation="0,0,0", Rotation=@n19, Scale="1,1,1") {rotationMode=BMM_Additive, rotationSpace=BCS_BoneSpace, alphaBoolBlend {blendOption=Linear}, boneToModify {boneName=dyn_cape_mid_01}}
   n23 = Get BwdMovement() -> (BwdMovement:double)
   n24 = Trail controller - Bone: dyn_cape_l_06!(ComponentPose=@n22, Alpha=@n23, RelaxationSpeedScale:float=1, FakeVelocity:Vector="0,0,0") -> (Pose) {
    bUsePlanarLimit=true
    chainLength=5
    debugLifeTime=0.1
    maxDeltaTime=0.0333
    relaxationSpeedScale=1.2
    alphaScaleBiasClamp {bInterpResult=true}
    rotationLimits [
     {limitMax=<180 90 180>, limitMin=<-180 -90 -180>}
     {limitMax=<180 90 180>, limitMin=<-180 -90 -180>}
     {limitMax=<180 90 180>, limitMin=<-180 -90 -180>}
     {limitMax=<180 90 180>, limitMin=<-180 -90 -180>}
     {limitMax=<180 90 180>, limitMin=<-180 -90 -180>}
    ]
    rotationOffsets [<0 0 0>, <0 0 0>, <0 0 0>, <0 0 0>, <0 0 0>]
    trailBone {boneName=dyn_cape_l_06}
    trailRelaxationSpeed {editorCurveData {keys [<at 0 15 RCIM_Linear>,<at 0.5289 13.1838 RCIM_Linear>,<at 1 6 RCIM_Linear>]}}
   }
   n25 = Trail controller - Bone: dyn_cape_mid_06!(ComponentPose=@n24, Alpha=@n23, RelaxationSpeedScale:float=1, FakeVelocity:Vector="0,0,0") -> (Pose) {
    bUsePlanarLimit=true
    chainLength=5
    debugLifeTime=0.1
    maxDeltaTime=0.0333
    relaxationSpeedScale=1.2
    alphaScaleBiasClamp {bInterpResult=true}
    rotationLimits [
     {limitMax=<180 180 180>}
     {limitMax=<180 180 180>}
     {limitMax=<180 90 180>, limitMin=<-180 -90 -180>}
     {limitMax=<180 180 180>}
     {limitMax=<180 180 180>}
    ]
    rotationOffsets [<0 0 0>, <0 0 0>, <0 0 0>, <0 0 0>, <0 0 0>]
    trailBone {boneName=dyn_cape_mid_06}
    trailRelaxationSpeed {editorCurveData {keys [<at 0 15 RCIM_Linear>,<at 0.5289 13.1838 RCIM_Linear>,<at 1 6 RCIM_Linear>]}}
   }
   n26 = Trail controller - Bone: dyn_cape_r_06!(ComponentPose=@n25, Alpha=@n23, RelaxationSpeedScale:float=1, FakeVelocity:Vector="0,0,0") -> (Pose) {
    bUsePlanarLimit=true
    chainLength=5
    debugLifeTime=0.1
    maxDeltaTime=0.0333
    relaxationSpeedScale=1.2
    alphaScaleBiasClamp {bInterpResult=true}
    rotationLimits [
     {limitMax=<180 90 180>, limitMin=<-180 -90 -180>}
     {limitMax=<180 90 180>, limitMin=<-180 -90 -180>}
     {limitMax=<180 90 180>, limitMin=<-180 -90 -180>}
     {limitMax=<180 90 180>, limitMin=<-180 -90 -180>}
     {limitMax=<180 90 180>, limitMin=<-180 -90 -180>}
    ]
    rotationOffsets [<0 0 0>, <0 0 0>, <0 0 0>, <0 0 0>, <0 0 0>]
    trailBone {boneName=dyn_cape_r_06}
    trailRelaxationSpeed {editorCurveData {keys [<at 0 15 RCIM_Linear>,<at 0.5289 13.1838 RCIM_Linear>,<at 1 6 RCIM_Linear>]}}
   }
   n27 = "Transform (Modify) Bone - Bone: dyn_cape_l_01"!(ComponentPose=@n26, Alpha=0, Translation="0,0,0", Rotation="5,0,0") -> (Pose) {
    rotationMode=BMM_Additive
    rotationSpace=BCS_BoneSpace
    translationMode=BMM_Additive
    translationSpace=BCS_BoneSpace
    alphaBoolBlend {blendOption=Linear}
    boneToModify {boneName=dyn_cape_l_01}
    rotation {pitch=5}
   }
   n28 = "Transform (Modify) Bone - Bone: dyn_cape_pin_r_01"!(ComponentPose=@n27, Alpha=1, Translation="0,0,0", Rotation="5,0,0") -> (Pose) {
    rotationMode=BMM_Additive
    rotationSpace=BCS_BoneSpace
    translationMode=BMM_Additive
    translationSpace=BCS_BoneSpace
    alphaBoolBlend {blendOption=Linear}
    boneToModify {boneName=dyn_cape_pin_r_01}
    rotation {pitch=5}
   }
   n29 = "Transform (Modify) Bone - Bone: dyn_cape_mid_01"!(ComponentPose=@n28, Alpha=0, Translation="0,0,0", Rotation="5,0,0") -> (Pose) {
    rotationMode=BMM_Additive
    rotationSpace=BCS_BoneSpace
    translationMode=BMM_Additive
    translationSpace=BCS_BoneSpace
    alphaBoolBlend {blendOption=Linear}
    boneToModify {boneName=dyn_cape_mid_01}
    rotation {pitch=5}
   }
   Save cached pose 'trailCache'(Pose=Component To Local(ComponentPose=@n29)) {cacheName=trailCache}
   n30 = "Transform (Modify) Bone - Bone: dyn_cape_l_01"!(
    ComponentPose=Local To Component(LocalPose=Use cached pose 'Start_Cache'())
    Alpha=1
    Translation="0,0,0"
    Rotation="5,0,0"
    Scale:Vector="1,1,1"
   ) -> (Pose) {
    rotationMode=BMM_Additive
    rotationSpace=BCS_BoneSpace
    alphaBoolBlend {blendOption=Linear}
    boneToModify {boneName=dyn_cape_l_01}
    rotation {pitch=5}
   }
   n31 = "Transform (Modify) Bone - Bone: dyn_cape_l_02"(ComponentPose=@n30, Alpha=1, Translation="0,0,0", Rotation="-15,0,0", Scale="1,1,1") {
    rotationMode=BMM_Additive
    rotationSpace=BCS_BoneSpace
    alphaBoolBlend {blendOption=Linear}
    boneToModify {boneName=dyn_cape_l_02}
    rotation {pitch=-15}
   }
   n32 = "Transform (Modify) Bone - Bone: dyn_cape_l_03"(ComponentPose=@n31, Alpha=1, Translation="0,0,0", Rotation="-15,0,0", Scale="1,1,1") {
    rotationMode=BMM_Additive
    rotationSpace=BCS_BoneSpace
    alphaBoolBlend {blendOption=Linear}
    boneToModify {boneName=dyn_cape_l_03}
    rotation {pitch=-15}
   }
   n33 = "Transform (Modify) Bone - Bone: dyn_cape_l_04"(ComponentPose=@n32, Alpha=1, Translation="0,0,0", Rotation="-25,15,0", Scale="1,1,1") {
    rotationMode=BMM_Additive
    rotationSpace=BCS_BoneSpace
    alphaBoolBlend {blendOption=Linear}
    boneToModify {boneName=dyn_cape_l_04}
    rotation {pitch=-25, yaw=15}
   }
   n34 = "Transform (Modify) Bone - Bone: dyn_cape_r_01"(ComponentPose=@n33, Alpha=1, Translation="0,0,0", Rotation="5,0,0", Scale="1,1,1") {
    rotationMode=BMM_Additive
    rotationSpace=BCS_BoneSpace
    alphaBoolBlend {blendOption=Linear}
    boneToModify {boneName=dyn_cape_r_01}
    rotation {pitch=5}
   }
   n35 = "Transform (Modify) Bone - Bone: dyn_cape_r_02"(ComponentPose=@n34, Alpha=1, Translation="0,0,0", Rotation="-15,0,0", Scale="1,1,1") {
    rotationMode=BMM_Additive
    rotationSpace=BCS_BoneSpace
    alphaBoolBlend {blendOption=Linear}
    boneToModify {boneName=dyn_cape_r_02}
    rotation {pitch=-15}
   }
   n36 = "Transform (Modify) Bone - Bone: dyn_cape_r_03"(ComponentPose=@n35, Alpha=1, Translation="0,0,0", Rotation="-15,0,0", Scale="1,1,1") {
    rotationMode=BMM_Additive
    rotationSpace=BCS_BoneSpace
    alphaBoolBlend {blendOption=Linear}
    boneToModify {boneName=dyn_cape_r_03}
    rotation {pitch=-15}
   }
   n37 = "Transform (Modify) Bone - Bone: dyn_cape_r_04"(ComponentPose=@n36, Alpha=1, Translation="0,0,0", Rotation="-25,-15,0", Scale="1,1,1") {
    rotationMode=BMM_Additive
    rotationSpace=BCS_BoneSpace
    alphaBoolBlend {blendOption=Linear}
    boneToModify {boneName=dyn_cape_r_04}
    rotation {pitch=-25, yaw=-15}
   }
   n38 = "Transform (Modify) Bone - Bone: dyn_cape_mid_01"!(ComponentPose=@n37, Alpha=1, Translation="0,0,0", Rotation="5,0,0", Scale:Vector="1,1,1") -> (Pose) {
    rotationMode=BMM_Additive
    rotationSpace=BCS_BoneSpace
    alphaBoolBlend {blendOption=Linear}
    boneToModify {boneName=dyn_cape_mid_01}
    rotation {pitch=5}
   }
   n39 = "Transform (Modify) Bone - Bone: dyn_cape_mid_02"(ComponentPose=@n38, Alpha=1, Translation="0,0,0", Rotation="-15,0,0", Scale="1,1,1") {
    rotationMode=BMM_Additive
    rotationSpace=BCS_BoneSpace
    alphaBoolBlend {blendOption=Linear}
    boneToModify {boneName=dyn_cape_mid_02}
    rotation {pitch=-15}
   }
   n40 = "Transform (Modify) Bone - Bone: dyn_cape_mid_03"(ComponentPose=@n39, Alpha=1, Translation="0,0,0", Rotation="-15,0,0", Scale="1,1,1") {
    rotationMode=BMM_Additive
    rotationSpace=BCS_BoneSpace
    alphaBoolBlend {blendOption=Linear}
    boneToModify {boneName=dyn_cape_mid_03}
    rotation {pitch=-15}
   }
   n41 = "Transform (Modify) Bone - Bone: dyn_cape_mid_04"(ComponentPose=@n40, Alpha=1, Translation="0,0,0", Rotation="-25,0.000,0.000", Scale="1,1,1") {
    rotationMode=BMM_Additive
    rotationSpace=BCS_BoneSpace
    alphaBoolBlend {blendOption=Linear}
    boneToModify {boneName=dyn_cape_mid_04}
    rotation {pitch=-25}
   }
   n42 = Trail controller - Bone: dyn_cape_mid_06(ComponentPose=@n41, Alpha=1, RelaxationSpeedScale=2, FakeVelocity="0,0,0") {
    bActorSpaceFakeVel=true
    bShowTrailLocation=true
    bUsePlanarLimit=true
    chainLength=5
    debugLifeTime=0.1
    maxDeltaTime=0.0333
    relaxationSpeedScale=1.2
    alphaScaleBiasClamp {bInterpResult=true, interpSpeedDecreasing=25, interpSpeedIncreasing=25}
    baseJoint {boneName=root}
    rotationLimits [{limitMax=<180 180 180>}, {limitMax=<180 180 180>}, {limitMax=<180 180 180>}, {limitMax=<180 180 180>}, {limitMax=<180 180 180>}]
    rotationOffsets [<0 0 0>, <0 0 0>, <0 0 0>, <0 0 0>, <0 0 0>]
    trailBone {boneName=dyn_cape_mid_06}
    trailRelaxationSpeed {editorCurveData {keys [<at 0 15 RCIM_Linear>,<at 1 6 RCIM_Linear>]}}
   }
   n43 = Trail controller - Bone: dyn_cape_l_06(ComponentPose=@n42, Alpha=1, RelaxationSpeedScale=2, FakeVelocity="0,0,0") {
    bActorSpaceFakeVel=true
    bShowTrailLocation=true
    bUsePlanarLimit=true
    chainLength=6
    debugLifeTime=0.1
    maxDeltaTime=0.0333
    relaxationSpeedScale=1.2
    alphaScaleBiasClamp {bInterpResult=true, interpSpeedDecreasing=25, interpSpeedIncreasing=25}
    baseJoint {boneName=root}
    rotationLimits [
     {limitMax=<180 180 180>}
     {limitMax=<180 180 180>}
     {limitMax=<180 180 180>}
     {limitMax=<180 180 180>}
     {limitMax=<180 180 180>}
     {limitMax=<180 180 180>}
    ]
    rotationOffsets [<0 0 0>, <0 0 0>, <0 0 0>, <0 0 0>, <0 0 0>, <0 0 0>]
    trailBone {boneName=dyn_cape_l_06}
    trailRelaxationSpeed {editorCurveData {keys [<at 0 15 RCIM_Linear>,<at 1 6 RCIM_Linear>]}}
   }
   n44 = Trail controller - Bone: dyn_cape_r_06(ComponentPose=@n43, Alpha=1, RelaxationSpeedScale=2, FakeVelocity="0,0,0") {
    bActorSpaceFakeVel=true
    bShowTrailLocation=true
    bUsePlanarLimit=true
    chainLength=6
    debugLifeTime=0.1
    maxDeltaTime=0.0333
    relaxationSpeedScale=1.2
    alphaScaleBiasClamp {bInterpResult=true, interpSpeedDecreasing=25, interpSpeedIncreasing=25}
    baseJoint {boneName=root}
    rotationLimits [
     {limitMax=<180 180 180>}
     {limitMax=<180 180 180>}
     {limitMax=<180 180 180>}
     {limitMax=<180 180 180>}
     {limitMax=<180 180 180>}
     {limitMax=<180 180 180>}
    ]
    rotationOffsets [<0 0 0>, <0 0 0>, <0 0 0>, <0 0 0>, <0 0 0>, <0 0 0>]
    trailBone {boneName=dyn_cape_r_06}
    trailRelaxationSpeed {editorCurveData {keys [<at 0 15 RCIM_Linear>,<at 1 6 RCIM_Linear>]}}
   }
   Save cached pose 'dbnoCache'(Pose=Component To Local(ComponentPose=@n44)) {cacheName=dbnoCache}
   n45 = Control Rig(Source=Copy Pose From Mesh() {bUseAttachedParent=true}) {
    alphaBoolBlend {blendOption=Linear}
    controlRigAssetReference {blueprintRigClass=&/F_MED_Obsidian_Deform_CtrlRig_twists.F_MED_Obsidian_Deform_CtrlRig_twists_C}
   }
   n46 = Control Rig!(
    Source=@n45
    Alpha
    KneeRot_fr_l_114deg:Vector
    KneeTrans_fr_l_114deg:Vector="0.258,-3.591,0.009"
    KneeScale_fr_l_114deg:Vector="1,1,1"
    KneeScale_fr_l_130deg:Vector="1,1,1"
    KneeRot_fr_l_130deg:Vector="0,0,-51.2"
    KneeTrans_fr_l_130deg:Vector="0.258,-3.591,0.009"
   ) -> (Pose) {
    alphaBoolBlend {blendOption=Linear}
    controlRigAssetReference {blueprintRigClass=&/F_MED_Obsidian_Deform_CtrlRig_sculpters.F_MED_Obsidian_Deform_CtrlRig_sculpters_C}
   }
   n47 = "Transform (Modify) Bone - Bone: dyn_root_slope"!(
    ComponentPose=Local To Component(LocalPose=@n46)
    bAlphaBoolEnabled:bool=NOT Boolean(A=Get bIsInFrontEnd())
    Translation=Get slope_root_trans() -> (slope_root_trans:Vector)
    Rotation=Get slope_foot_rot() -> (slope_foot_rot:Rotator)
    Scale="1,1,1"
   ) -> (Pose) {
    alphaInputType=Bool
    rotationMode=BMM_Replace
    rotationSpace=BCS_WorldSpace
    translationMode=BMM_Replace
    translationSpace=BCS_WorldSpace
    alphaBoolBlend {blendOption=Linear}
    boneToModify {boneName=dyn_root_slope}
   }
   n48 = Get thigh_rot_up_l() -> (thigh_rot_up_l:double)
   n49 = "Transform (Modify) Bone - Bone: dyn_loin_MSTR"!(ComponentPose=@n47, Alpha:float=@n48, Translation="0,0,0", Rotation="15,0,-5", Scale="1,1,1") -> (Pose) {
    rotationMode=BMM_Additive
    rotationSpace=BCS_BoneSpace
    translationMode=BMM_Additive
    translationSpace=BCS_BoneSpace
    alphaBoolBlend {blendOption=Linear}
    boneToModify {boneName=dyn_loin_MSTR}
    rotation {pitch=15, roll=-5}
   }
   n50 = "Transform (Modify) Bone - Bone: dyn_loin_l_00"(ComponentPose=@n49, Alpha=@n48, Translation="-3.575,12.736,0.000", Rotation="0.000,-25.000,0.000", Scale="1,1,1") {
    rotationMode=BMM_Additive
    rotationSpace=BCS_BoneSpace
    translationMode=BMM_Additive
    translationSpace=BCS_BoneSpace
    alphaBoolBlend {blendOption=Linear}
    boneToModify {boneName=dyn_loin_l_00}
    rotation {yaw=-24.9997}
    translation=<-3.575 12.736>
   }
   n51 = Get thigh_rot_up_r() -> (thigh_rot_up_r:double)
   n52 = "Transform (Modify) Bone - Bone: dyn_loin_MSTR"(ComponentPose=@n50, Alpha=@n51, Translation="0,0,0", Rotation="15,0,5", Scale="1,1,1") {
    rotationMode=BMM_Additive
    rotationSpace=BCS_BoneSpace
    translationMode=BMM_Additive
    translationSpace=BCS_BoneSpace
    alphaBoolBlend {blendOption=Linear}
    boneToModify {boneName=dyn_loin_MSTR}
    rotation {pitch=15, roll=5}
   }
   n53 = "Transform (Modify) Bone - Bone: dyn_roin_r_00"(ComponentPose=@n52, Alpha=@n51, Translation="-2.234,-18.778,0.000", Rotation="0.000,25.000,0.000", Scale="1,1,1") {
    rotationMode=BMM_Additive
    rotationSpace=BCS_BoneSpace
    translationMode=BMM_Additive
    translationSpace=BCS_BoneSpace
    alphaBoolBlend {blendOption=Linear}
    boneToModify {boneName=dyn_roin_r_00}
    rotation {yaw=24.9998}
    translation=<-2.234 -18.778>
   }
   n54 = "Transform (Modify) Bone - Bone: dyn_belt_discs_fr_mid_r"(ComponentPose=@n53, Alpha=@n51, Translation="-3,0.803,2.494", Rotation="-13.200,-15.358,8.762", Scale="1,1,1") {
    rotationMode=BMM_Additive
    rotationSpace=BCS_BoneSpace
    translationMode=BMM_Additive
    translationSpace=BCS_BoneSpace
    alphaBoolBlend {blendOption=Linear}
    boneToModify {boneName=dyn_belt_discs_fr_mid_r}
    rotation=<rot -13.2 -15.3582 8.7621>
    translation=<-3 0.803 2.494>
   }
   n55 = "Transform (Modify) Bone - Bone: dyn_loin_MSTR"!(ComponentPose=@n54, bAlphaBoolEnabled:bool=Get is_crouch_moving(), Translation="-4,0,6", Rotation="70,0,0", Scale="1,1,1") -> (Pose) {
    alphaInputType=Bool
    rotationMode=BMM_Additive
    rotationSpace=BCS_BoneSpace
    translationMode=BMM_Additive
    translationSpace=BCS_BoneSpace
    alphaBoolBlend {blendInTime=0.2, blendOption=Linear, blendOutTime=0.2}
    boneToModify {boneName=dyn_loin_MSTR}
    rotation {pitch=70}
    translation {x=-4, z=6}
   }
   n56 = "Transform (Modify) Bone - Bone: dyn_loin_MSTR"!(ComponentPose=@n55, Alpha:float=Get is_walk_moving() -> (is_walk_moving:double), Translation="0,0,0", Rotation="40,0,0", Scale="1,1,1") -> (Pose) {
    rotationMode=BMM_Additive
    rotationSpace=BCS_BoneSpace
    translationMode=BMM_Additive
    translationSpace=BCS_BoneSpace
    alphaBoolBlend {blendInTime=0.2, blendOption=Linear, blendOutTime=0.2}
    alphaScaleBiasClamp {bInterpResult=true, interpSpeedDecreasing=0.7, interpSpeedIncreasing=0.7}
    boneToModify {boneName=dyn_loin_MSTR}
    rotation {pitch=40}
   }
   n57 = "Constraint - dyn_loin_02 (PP)"!(
    ComponentPose=@n56
    Alpha=Get CrouchingMoving2D - Constraints() -> (CrouchingMoving2D - Constraints:double)
    ConstraintWeights_0=0.5
    ConstraintWeights_1:float=0.5
   ) -> (Pose) {
    alphaBoolBlend {blendOption=Linear}
    alphaScaleBiasClamp {bInterpResult=true, interpSpeedDecreasing=5, interpSpeedIncreasing=5}
    boneToModify {boneName=dyn_loin_02}
    constraintSetup [{targetBone {boneName=calf_r}}, {targetBone {boneName=calf_l}}]
    constraintWeights [1,1]
   }
   n58 = Get CrouchingMoving2D() -> (CrouchingMoving2D:double)
   n59 = "Transform (Modify) Bone - Bone: dyn_loin_MSTR"(ComponentPose=@n57, Alpha=@n58, Translation="-6.151,-0.000,4.771", Rotation="10.000,-0.000,-0.000", Scale="1,1,1") {
    rotationMode=BMM_Additive
    rotationSpace=BCS_BoneSpace
    translationMode=BMM_Additive
    translationSpace=BCS_BoneSpace
    alphaBoolBlend {blendOption=Linear}
    alphaScaleBiasClamp {bInterpResult=true, interpSpeedDecreasing=5, interpSpeedIncreasing=5}
    boneToModify {boneName=dyn_loin_MSTR}
    rotation=<rot 10 0 0>
    translation {x=-6.151, z=4.771}
   }
   n60 = "Transform (Modify) Bone - Bone: dyn_loin_03"!(ComponentPose=@n59, Alpha:float=@n58, Translation="0,0,0", Rotation="15,0,0", Scale="1,1,1") -> (Pose) {
    rotationMode=BMM_Additive
    rotationSpace=BCS_BoneSpace
    translationMode=BMM_Additive
    translationSpace=BCS_BoneSpace
    alphaBoolBlend {blendOption=Linear}
    alphaScaleBiasClamp {bInterpResult=true, interpSpeedDecreasing=5, interpSpeedIncreasing=5}
    boneToModify {boneName=dyn_loin_03}
    rotation {pitch=15}
   }
   n61 = "Transform (Modify) Bone - Bone: dyn_loin_03"(ComponentPose=@n60, Alpha=Get IddleCrouch() -> (IddleCrouch:double), Translation="2.478,0.000,-2.078", Rotation="45,0,0", Scale="1,1,1") {
    rotationMode=BMM_Additive
    rotationSpace=BCS_BoneSpace
    translationMode=BMM_Additive
    translationSpace=BCS_BoneSpace
    alphaBoolBlend {blendOption=Linear}
    alphaScaleBiasClamp {bInterpResult=true, interpSpeedDecreasing=5, interpSpeedIncreasing=5}
    boneToModify {boneName=dyn_loin_03}
    rotation {pitch=45}
    translation=<2.4777 0 -2.0778>
   }
   n62 = "Transform (Modify) Bone - Bone: dyn_loin_03"!(ComponentPose=@n61, bAlphaBoolEnabled:bool=True, Translation="0,0,0", Rotation="20.000,-0.000,0.000", Scale="1,1,1") -> (Pose) {
    alphaInputType=Bool
    rotationMode=BMM_Additive
    rotationSpace=BCS_BoneSpace
    translationMode=BMM_Additive
    translationSpace=BCS_BoneSpace
    alphaBoolBlend {blendInTime=0.1, blendOption=Linear, blendOutTime=0.2}
    alphaScaleBiasClamp {bInterpResult=true, interpSpeedDecreasing=5, interpSpeedIncreasing=5}
    boneToModify {boneName=dyn_loin_03}
    rotation=<rot 20.0002 0 0>
   }
   n63 = "Transform (Modify) Bone - Bone: dyn_loin_02"(
    ComponentPose=@n62
    Alpha=Get is_SprintingNotSliding() -> (is_SprintingNotSliding:double)
    Translation="-8.927,0.018,8.147"
    Rotation="0,0,0"
    Scale="1,1,1"
   ) {
    rotationMode=BMM_Additive
    rotationSpace=BCS_BoneSpace
    translationMode=BMM_Additive
    translationSpace=BCS_BoneSpace
    alphaBoolBlend {blendInTime=0.2, blendOption=Linear, blendOutTime=0.2}
    alphaScaleBiasClamp {bInterpResult=true, interpSpeedDecreasing=5, interpSpeedIncreasing=5}
    boneToModify {boneName=dyn_loin_02}
    translation=<-8.927 0.018 8.147>
   }
   n64 = Get Moving2D_On_Ground() -> (Moving2D_On_Ground:double)
   n65 = "Constraint - dyn_loin_01 (PP)"!(ComponentPose=@n63, Alpha=@n64, ConstraintWeights_0=0.5, ConstraintWeights_1:float=0.5) -> (Pose) {
    alphaBoolBlend {blendInTime=0.2, blendOption=Linear, blendOutTime=0.2}
    alphaScaleBiasClamp {bInterpResult=true, interpSpeedDecreasing=5, interpSpeedIncreasing=5}
    boneToModify {boneName=dyn_loin_01}
    constraintSetup [{targetBone {boneName=calf_l}}, {targetBone {boneName=calf_r}}]
    constraintWeights [1,1]
   }
   n66 = "Transform (Modify) Bone - Bone: dyn_loin_02"(ComponentPose=@n65, Alpha=@n64, Translation="0.486,0.000,-1.045", Rotation="25.000,-0.000,-0.000", Scale="1,1,1") {
    rotationMode=BMM_Additive
    rotationSpace=BCS_BoneSpace
    translationMode=BMM_Additive
    translationSpace=BCS_BoneSpace
    alphaBoolBlend {blendInTime=0.2, blendOption=Linear, blendOutTime=0.2}
    alphaScaleBiasClamp {bInterpResult=true, interpSpeedDecreasing=5, interpSpeedIncreasing=5}
    boneToModify {boneName=dyn_loin_02}
    rotation=<rot 25 -0.0001 0>
    translation {x=0.486, z=-1.045}
   }
   n67 = "Constraint - dyn_belt_discs_bk_MSTR (PPP)"!(ComponentPose=@n66, Alpha=1, ConstraintWeights_0=0.8073, ConstraintWeights_1:float=0.0438, ConstraintWeights_2:float=0.1489) -> (Pose) {
    alphaBoolBlend {blendOption=Linear}
    boneToModify {boneName=dyn_belt_discs_bk_MSTR}
    constraintSetup [{targetBone {boneName=pelvis}}, {targetBone {boneName=thigh_twist_01_l}}, {targetBone {boneName=deform_thigh_bk_l}}]
    constraintWeights [1,1,1]
   }
   n68 = "Constraint - dyn_belt_discs_bk_l_MSTR (PPP)"!(ComponentPose=@n67, Alpha=1, ConstraintWeights_0=0.6772, ConstraintWeights_1:float=0.1282, ConstraintWeights_2:float=0.1946) -> (Pose) {
    alphaBoolBlend {blendOption=Linear}
    boneToModify {boneName=dyn_belt_discs_bk_l_MSTR}
    constraintSetup [{targetBone {boneName=pelvis}}, {targetBone {boneName=thigh_twist_01_l}}, {targetBone {boneName=deform_thigh_bk_l}}]
    constraintWeights [1,1,1]
   }
   n69 = "Constraint - dyn_belt_discs_l_MSTR (PPPPP)"!(
    ComponentPose=@n68
    Alpha=1
    ConstraintWeights_0=0.4713
    ConstraintWeights_1:float=0.2965
    ConstraintWeights_2:float=0.129
    ConstraintWeights_3:float=0.0059
    ConstraintWeights_4:float=0.0973
   ) -> (Pose) {
    alphaBoolBlend {blendOption=Linear}
    boneToModify {boneName=dyn_belt_discs_l_MSTR}
    constraintSetup [
     {targetBone {boneName=pelvis}}
     {targetBone {boneName=thigh_twist_01_l}}
     {targetBone {boneName=deform_thigh_bk_l}}
     {targetBone {boneName=thigh_l}}
     {targetBone {boneName=deform_thigh_fr_l}}
    ]
    constraintWeights [1,1,1,1,1]
   }
   n70 = "Constraint - dyn_belt_discs_fr_mid_l_MSTR (PPPPP)"!(
    ComponentPose=@n69
    Alpha=1
    ConstraintWeights_0=0.3275
    ConstraintWeights_1:float=0.289
    ConstraintWeights_2:float=0.0143
    ConstraintWeights_3:float=0.0393
    ConstraintWeights_4:float=0.3298
   ) -> (Pose) {
    alphaBoolBlend {blendOption=Linear}
    boneToModify {boneName=dyn_belt_discs_fr_mid_l_MSTR}
    constraintSetup [
     {targetBone {boneName=pelvis}}
     {targetBone {boneName=thigh_twist_01_l}}
     {targetBone {boneName=deform_thigh_bk_l}}
     {targetBone {boneName=thigh_l}}
     {targetBone {boneName=deform_thigh_fr_l}}
    ]
    constraintWeights [1,1,1,1,1]
   }
   n71 = "Constraint - dyn_belt_discs_fr_l_MSTR (PPPPPPP)"!(
    ComponentPose=@n70
    Alpha=1
    ConstraintWeights_0=0.338
    ConstraintWeights_1:float=0
    ConstraintWeights_2:float=0.3117
    ConstraintWeights_3:float=0.0364
    ConstraintWeights_4:float=0.3139
    ConstraintWeights_5:float=0
    ConstraintWeights_6:float=0
   ) -> (Pose) {
    alphaBoolBlend {blendOption=Linear}
    boneToModify {boneName=dyn_belt_discs_fr_l_MSTR}
    constraintSetup [
     {targetBone {boneName=deform_thigh_fr_l}}
     {targetBone {boneName=thigh_twist_01_r}}
     {targetBone {boneName=thigh_twist_01_l}}
     {targetBone {boneName=thigh_l}}
     {targetBone {boneName=pelvis}}
     {targetBone {boneName=deform_thigh_bk_l}}
     {targetBone {boneName=deform_thigh_fr_r}}
    ]
    constraintWeights [1,1,1,1,1,1,1]
   }
   n72 = "Constraint - dyn_belt_discs_fr_fr_l_MSTR (PPPP)"!(
    ComponentPose=@n71
    Alpha=1
    ConstraintWeights_0=0.6414
    ConstraintWeights_1:float=0.073
    ConstraintWeights_2:float=0.2327
    ConstraintWeights_3:float=0.0529
   ) -> (Pose) {
    alphaBoolBlend {blendOption=Linear}
    boneToModify {boneName=dyn_belt_discs_fr_fr_l_MSTR}
    constraintSetup [
     {targetBone {boneName=pelvis}}
     {targetBone {boneName=thigh_twist_01_l}}
     {targetBone {boneName=deform_thigh_fr_l}}
     {targetBone {boneName=deform_thigh_fr_r}}
    ]
    constraintWeights [1,1,1,1]
   }
   n73 = "Constraint - dyn_belt_discs_fr_MSTR (PPPP)"!(
    ComponentPose=@n72
    Alpha=1
    ConstraintWeights_0=0.0422
    ConstraintWeights_1:float=0.7205
    ConstraintWeights_2:float=0.039
    ConstraintWeights_3:float=0.1983
   ) -> (Pose) {
    alphaBoolBlend {blendOption=Linear}
    boneToModify {boneName=dyn_belt_discs_fr_MSTR}
    constraintSetup [
     {targetBone {boneName=deform_thigh_fr_l}}
     {targetBone {boneName=pelvis}}
     {targetBone {boneName=thigh_twist_01_r}}
     {targetBone {boneName=deform_thigh_fr_r}}
    ]
    constraintWeights [1,1,1,1]
   }
   n74 = "Constraint - dyn_belt_discs_fr_mid_r_MSTR (PPPPP)"!(
    ComponentPose=@n73
    Alpha=1
    ConstraintWeights_0=0.7245
    ConstraintWeights_1:float=0.0276
    ConstraintWeights_2:float=0.0143
    ConstraintWeights_3:float=0.0439
    ConstraintWeights_4:float=0.1896
   ) -> (Pose) {
    alphaBoolBlend {blendOption=Linear}
    boneToModify {boneName=dyn_belt_discs_fr_mid_r_MSTR}
    constraintSetup [
     {targetBone {boneName=pelvis}}
     {targetBone {boneName=spine_02}}
     {targetBone {boneName=spine_01}}
     {targetBone {boneName=thigh_twist_01_r}}
     {targetBone {boneName=deform_thigh_fr_r}}
    ]
    constraintWeights [1,1,1,1,1]
   }
   n75 = "Constraint - dyn_belt_discs_fr_r_MSTR (PPPPPP)"!(
    ComponentPose=@n74
    Alpha=1
    ConstraintWeights_0=0.0304
    ConstraintWeights_1:float=0.0557
    ConstraintWeights_2:float=0.095
    ConstraintWeights_3:float=0.0196
    ConstraintWeights_4:float=0.6803
    ConstraintWeights_5:float=0.1191
   ) -> (Pose) {
    alphaBoolBlend {blendOption=Linear}
    boneToModify {boneName=dyn_belt_discs_fr_r_MSTR}
    constraintSetup [
     {targetBone {boneName=thigh_twist_01_r}}
     {targetBone {boneName=spine_01}}
     {targetBone {boneName=spine_02}}
     {targetBone {boneName=spine_04}}
     {targetBone {boneName=pelvis}}
     {targetBone {boneName=deform_thigh_fr_r}}
    ]
    constraintWeights [1,1,1,1,1,1]
   }
   n76 = "Constraint - dyn_belt_discs_r_MSTR (PPPPPPP)"!(
    ComponentPose=@n75
    Alpha=1
    ConstraintWeights_0=0
    ConstraintWeights_1:float=0.0294
    ConstraintWeights_2:float=0.0824
    ConstraintWeights_3:float=0.1268
    ConstraintWeights_4:float=0.0221
    ConstraintWeights_5:float=0.6397
    ConstraintWeights_6:float=0.0996
   ) -> (Pose) {
    alphaBoolBlend {blendOption=Linear}
    boneToModify {boneName=dyn_belt_discs_r_MSTR}
    constraintSetup [
     {targetBone {boneName=deform_thigh_bk_r}}
     {targetBone {boneName=thigh_twist_01_r}}
     {targetBone {boneName=spine_01}}
     {targetBone {boneName=spine_02}}
     {targetBone {boneName=spine_04}}
     {targetBone {boneName=pelvis}}
     {targetBone {boneName=deform_thigh_fr_r}}
    ]
    constraintWeights [1,1,1,1,1,1,1]
   }
   n77 = "Constraint - dyn_belt_discs_bk_bk_r_MSTR (PPPPP)"!(
    ComponentPose=@n76
    Alpha=1
    ConstraintWeights_0=0.7629
    ConstraintWeights_1:float=0.0518
    ConstraintWeights_2:float=0.0912
    ConstraintWeights_3:float=0.0139
    ConstraintWeights_4:float=0.0802
   ) -> (Pose) {
    alphaBoolBlend {blendOption=Linear}
    boneToModify {boneName=dyn_belt_discs_bk_bk_r_MSTR}
    constraintSetup [
     {targetBone {boneName=pelvis}}
     {targetBone {boneName=deform_thigh_bk_r}}
     {targetBone {boneName=spine_02}}
     {targetBone {boneName=thigh_twist_01_r}}
     {targetBone {boneName=spine_01}}
    ]
    constraintWeights [1,1,1,1,1]
   }
   n78 = "Constraint - dyn_belt_discs_bk_r_r_MSTR (PPPP)"!(
    ComponentPose=@n77
    Alpha=1
    ConstraintWeights_0=0.7936
    ConstraintWeights_1:float=0.0626
    ConstraintWeights_2:float=0.0716
    ConstraintWeights_3:float=0.0722
   ) -> (Pose) {
    alphaBoolBlend {blendOption=Linear}
    boneToModify {boneName=dyn_belt_discs_bk_r_r_MSTR}
    constraintSetup [
     {targetBone {boneName=pelvis}}
     {targetBone {boneName=deform_thigh_bk_r}}
     {targetBone {boneName=spine_02}}
     {targetBone {boneName=spine_01}}
    ]
    constraintWeights [1,1,1,1]
   }
   n79 = "Constraint - dyn_belt_discs_bk_mid_r_MSTR (PPPPP)"!(
    ComponentPose=@n78
    Alpha=1
    ConstraintWeights_0=0.8414
    ConstraintWeights_1:float=0.0764
    ConstraintWeights_2:float=0.0288
    ConstraintWeights_3:float=0.016
    ConstraintWeights_4:float=0.0374
   ) -> (Pose) {
    alphaBoolBlend {blendOption=Linear}
    boneToModify {boneName=dyn_belt_discs_bk_mid_r_MSTR}
    constraintSetup [
     {targetBone {boneName=pelvis}}
     {targetBone {boneName=deform_thigh_bk_r}}
     {targetBone {boneName=spine_02}}
     {targetBone {boneName=thigh_twist_01_r}}
     {targetBone {boneName=spine_01}}
    ]
    constraintWeights [1,1,1,1,1]
   }
   n80 = "Constraint - dyn_cape_pin_bk_out_l_01_MSTR (R)"(ComponentPose=@n79, Alpha=1, ConstraintWeights_0=1) {
    alphaBoolBlend {blendOption=Linear}
    boneToModify {boneName=dyn_cape_pin_bk_out_l_01_MSTR}
    constraintSetup [{transformType=Rotation, targetBone {boneName=spine_05}}]
    constraintWeights [1]
   }
   n81 = "Constraint - dyn_cape_pin_bk_out_r_01_MSTR (R)"(ComponentPose=@n80, Alpha=1, ConstraintWeights_0=1) {
    alphaBoolBlend {blendOption=Linear}
    boneToModify {boneName=dyn_cape_pin_bk_out_r_01_MSTR}
    constraintSetup [{transformType=Rotation, targetBone {boneName=spine_05}}]
    constraintWeights [1]
   }
   n82 = "Constraint - dyn_belt_discs_fr_l_MSTR (PPPP)"!(
    ComponentPose=@n81
    Alpha=1
    ConstraintWeights_0=0.337
    ConstraintWeights_1:float=0.2295
    ConstraintWeights_2:float=0.0116
    ConstraintWeights_3:float=0.4219
   ) -> (Pose) {
    alphaBoolBlend {blendOption=Linear}
    boneToModify {boneName=dyn_belt_discs_fr_l_MSTR}
    constraintSetup [
     {targetBone {boneName=deform_thigh_fr_l}}
     {targetBone {boneName=thigh_twist_01_l}}
     {targetBone {boneName=thigh_l}}
     {targetBone {boneName=pelvis}}
    ]
    constraintWeights [1,1,1,1]
   }
   n83 = "Constraint - dyn_belt_discs_fr_l_tgt_MSTR (PPPPPPP)"!(
    ComponentPose=@n82
    Alpha=1
    ConstraintWeights_0=0.2768
    ConstraintWeights_1:float=0
    ConstraintWeights_2:float=0.4775
    ConstraintWeights_3:float=0.1181
    ConstraintWeights_4:float=0.1276
    ConstraintWeights_5:float=0
    ConstraintWeights_6:float=0
   ) -> (Pose) {
    alphaBoolBlend {blendOption=Linear}
    boneToModify {boneName=dyn_belt_discs_fr_l_tgt_MSTR}
    constraintSetup [
     {targetBone {boneName=deform_thigh_fr_l}}
     {targetBone {boneName=thigh_twist_01_r}}
     {targetBone {boneName=thigh_twist_01_l}}
     {targetBone {boneName=thigh_l}}
     {targetBone {boneName=pelvis}}
     {targetBone {boneName=deform_thigh_bk_l}}
     {targetBone {boneName=deform_thigh_fr_r}}
    ]
    constraintWeights [1,1,1,1,1,1,1]
   }
   n84 = "Constraint - dyn_belt_discs_fr_l_up_MSTR (PPPPP)"!(
    ComponentPose=@n83
    Alpha=1
    ConstraintWeights_0=0.318
    ConstraintWeights_1:float=0.399
    ConstraintWeights_2:float=0.0004
    ConstraintWeights_3:float=0.0777
    ConstraintWeights_4:float=0.205
   ) -> (Pose) {
    alphaBoolBlend {blendOption=Linear}
    boneToModify {boneName=dyn_belt_discs_fr_l_up_MSTR}
    constraintSetup [
     {targetBone {boneName=deform_thigh_fr_l}}
     {targetBone {boneName=thigh_twist_01_l}}
     {targetBone {boneName=deform_thigh_bk_l}}
     {targetBone {boneName=thigh_l}}
     {targetBone {boneName=pelvis}}
    ]
    constraintWeights [1,1,1,1,1]
   }
   n85 = "Constraint - dyn_belt_discs_fr_mid_l_MSTR (PPPPP)"(
    ComponentPose=@n84
    Alpha=1
    ConstraintWeights_0=0.3242
    ConstraintWeights_1=0.2347
    ConstraintWeights_2=0.0133
    ConstraintWeights_3=0.0103
    ConstraintWeights_4=0.4173
   ) {
    alphaBoolBlend {blendOption=Linear}
    boneToModify {boneName=dyn_belt_discs_fr_mid_l_MSTR}
    constraintSetup [
     {targetBone {boneName=deform_thigh_fr_l}}
     {targetBone {boneName=thigh_twist_01_l}}
     {targetBone {boneName=deform_thigh_bk_l}}
     {targetBone {boneName=thigh_l}}
     {targetBone {boneName=pelvis}}
    ]
    constraintWeights [1,1,1,1,1]
   }
   n86 = "Constraint - dyn_belt_discs_fr_mid_l_tgt_MSTR (PPPPP)"!(
    ComponentPose=@n85
    Alpha=1
    ConstraintWeights_0=0.3001
    ConstraintWeights_1:float=0.3775
    ConstraintWeights_2:float=0.0113
    ConstraintWeights_3:float=0.0877
    ConstraintWeights_4:float=0.2234
   ) -> (Pose) {
    alphaBoolBlend {blendOption=Linear}
    boneToModify {boneName=dyn_belt_discs_fr_mid_l_tgt_MSTR}
    constraintSetup [
     {targetBone {boneName=deform_thigh_fr_l}}
     {targetBone {boneName=thigh_twist_01_l}}
     {targetBone {boneName=deform_thigh_bk_l}}
     {targetBone {boneName=thigh_l}}
     {targetBone {boneName=pelvis}}
    ]
    constraintWeights [1,1,1,1,1]
   }
   n87 = "Constraint - dyn_belt_discs_fr_mid_l_up_MSTR (PPPPP)"!(
    ComponentPose=@n86
    Alpha=1
    ConstraintWeights_0=0.2862
    ConstraintWeights_1:float=0.2927
    ConstraintWeights_2:float=0.036
    ConstraintWeights_3:float=0.0351
    ConstraintWeights_4:float=0.35
   ) -> (Pose) {
    alphaBoolBlend {blendOption=Linear}
    boneToModify {boneName=dyn_belt_discs_fr_mid_l_up_MSTR}
    constraintSetup [
     {targetBone {boneName=deform_thigh_fr_l}}
     {targetBone {boneName=thigh_twist_01_l}}
     {targetBone {boneName=deform_thigh_bk_l}}
     {targetBone {boneName=thigh_l}}
     {targetBone {boneName=pelvis}}
    ]
    constraintWeights [1,1,1,1,1]
   }
   n88 = Control Rig!(Source=Component To Local(ComponentPose=@n87), Alpha) -> (Pose) {
    alphaBoolBlend {blendOption=Linear}
    controlRigAssetReference {blueprintRigClass=&/F_MED_Obsidian_CtrlRig_medallions.F_MED_Obsidian_CtrlRig_medallions_C}
   }
   n89 = "Transform (Modify) Bone - Bone: dyn_belt_discs_fr_l"(ComponentPose=Local To Component(LocalPose=@n88), Alpha=Get thigh_rot_up_l(), Translation="0,-1,2", Rotation="0,-10,-10", Scale="1,1,1") {
    rotationMode=BMM_Additive
    rotationSpace=BCS_BoneSpace
    translationMode=BMM_Additive
    translationSpace=BCS_BoneSpace
    alphaBoolBlend {blendOption=Linear}
    boneToModify {boneName=dyn_belt_discs_fr_l}
    rotation {roll=-10, yaw=-10}
    translation {y=-1, z=2}
   }
   n90 = "Constraint - dyn_cape_disc_fr_MSTR (PP)"!(ComponentPose=@n89, Alpha=1, ConstraintWeights_0=0.2153, ConstraintWeights_1:float=0.7847) -> (Pose) {
    alphaBoolBlend {blendOption=Linear}
    boneToModify {boneName=dyn_cape_disc_fr_MSTR}
    constraintSetup [{targetBone {boneName=spine_04}}, {targetBone {boneName=spine_05}}]
    constraintWeights [1,1]
   }
   n91 = "Transform (Modify) Bone - Bone: dyn_root_simspace"(ComponentPose=@n90, Alpha=1, Translation="0,0,0", Rotation=Get sim_space_rotation() -> (sim_space_rotation:Rotator), Scale="1,1,1") {
    rotationMode=BMM_Replace
    rotationSpace=BCS_WorldSpace
    translationSpace=BCS_BoneSpace
    alphaBoolBlend {blendOption=Linear}
    alphaScaleBiasClamp {bInterpResult=true, interpSpeedDecreasing=2, interpSpeedIncreasing=2}
    boneToModify {boneName=dyn_root_simspace}
   }
   n92 = "Transform (Modify) Bone - Bone: dyn_crouch_collision"(
    ComponentPose=@n91
    Alpha=Get crouch alpha inv() -> (crouch alpha inv:double)
    Translation="0.000,0.000,-49.249"
    Rotation="0,0,0"
    Scale="1,1,1"
   ) {
    translationMode=BMM_Replace
    translationSpace=BCS_BoneSpace
    alphaBoolBlend {blendOption=Linear}
    alphaScaleBiasClamp {bInterpResult=true}
    boneToModify {boneName=dyn_crouch_collision}
    translation {z=-49.249}
   }
   Save cached pose 'Start_Cache'(Pose=Component To Local(ComponentPose=@n92)) {cacheName=Start_Cache}
   n93 = RigidBody(ComponentPose=Local To Component(LocalPose=Use cached pose 'Start_Cache'()), OverrideWorldGravity=Get GravityOverride()) {
    bClampLinearTranslationLimitToRefPose=true
    bOverrideWorldGravity=true
    overridePhysicsAsset=&/F_MED_Obsidian_Dyn_cape
    simulationSpace=BaseBoneSpace
    alphaBoolBlend {blendOption=Linear}
    baseBoneRef {boneName=dyn_root_simspace}
    componentAppliedLinearAccClamp=<3000 3000 3000>
    componentLinearAccScale {z=0.3}
    componentLinearVelScale {z=0.3}
    overrideWorldGravity {z=-980}
    simSpaceSettings {
     maxAngularAcceleration=500
     maxAngularVelocity=3
     maxLinearAcceleration=700
     maxLinearVelocity=500
     velocityScaleZ=0
     worldAlpha=1
     externalLinearDragV=<1 1>
    }
   }
   n94 = "Transform (Modify) Bone - Bone: dyn_cape_l_01"!(ComponentPose=@n93, Alpha=0, Translation="-10,0,8", Rotation="0,0,0") -> (Pose) {
    rotationMode=BMM_Additive
    rotationSpace=BCS_BoneSpace
    translationMode=BMM_Additive
    translationSpace=BCS_BoneSpace
    alphaBoolBlend {blendOption=Linear}
    boneToModify {boneName=dyn_cape_l_01}
    translation {x=-10, z=8}
   }
   n95 = "Transform (Modify) Bone - Bone: dyn_cape_pin_r_01"(ComponentPose=@n94, Alpha=1, Translation="-10,0,8", Rotation="0,0,0") {
    rotationMode=BMM_Additive
    rotationSpace=BCS_BoneSpace
    translationMode=BMM_Additive
    translationSpace=BCS_BoneSpace
    alphaBoolBlend {blendOption=Linear}
    boneToModify {boneName=dyn_cape_pin_r_01}
    translation {x=-10, z=8}
   }
   n96 = "Transform (Modify) Bone - Bone: dyn_cape_mid_01"!(ComponentPose=@n95, Alpha=0, Translation="-10,0,8", Rotation="0,0,0") -> (Pose) {
    rotationMode=BMM_Additive
    rotationSpace=BCS_BoneSpace
    translationMode=BMM_Additive
    translationSpace=BCS_BoneSpace
    alphaBoolBlend {blendOption=Linear}
    boneToModify {boneName=dyn_cape_mid_01}
    translation {x=-10, z=8}
   }
   n97 = Component To Local(ComponentPose=@n96)
   n98 = RigidBody(ComponentPose=Local To Component(LocalPose=Use cached pose 'Start_Cache'()), OverrideWorldGravity=Get GravityOverride()) {
    bClampLinearTranslationLimitToRefPose=true
    bOverrideWorldGravity=true
    overridePhysicsAsset=&F_MED_Obsidian_Dyn_cape
    simulationSpace=BaseBoneSpace
    alphaBoolBlend {blendOption=Linear}
    baseBoneRef {boneName=dyn_root_simspace}
    componentAppliedLinearAccClamp=<3000 3000 3000>
    componentLinearAccScale {z=0.3}
    componentLinearVelScale {z=0.3}
    overrideWorldGravity {z=-980}
    simSpaceSettings {
     maxAngularAcceleration=500
     maxAngularVelocity=3
     maxLinearAcceleration=700
     maxLinearVelocity=500
     velocityScaleZ=0
     worldAlpha=1
     externalLinearDragV=<1 1>
    }
   }
   n99 = Component To Local(ComponentPose=@n98)
   n100 = Get isWearingBackbling()
   n101 = RigidBody(ComponentPose=Local To Component(LocalPose=Use cached pose 'Start_Cache'()), OverrideWorldGravity="0,0,-980") {
    bClampLinearTranslationLimitToRefPose=true
    bOverrideWorldGravity=true
    overridePhysicsAsset=&/F_MED_Obsidian_Dyn_cape_SunburstEmote
    simulationSpace=BaseBoneSpace
    alphaBoolBlend {blendOption=Linear}
    baseBoneRef {boneName=dyn_root_simspace}
    componentAppliedLinearAccClamp=<3000 3000 3000>
    componentLinearAccScale {z=0.3}
    componentLinearVelScale {z=0.3}
    overrideWorldGravity {z=-980}
    simSpaceSettings {
     maxAngularAcceleration=500
     maxAngularVelocity=3
     maxLinearAcceleration=700
     maxLinearVelocity=500
     velocityScaleZ=0
     worldAlpha=1
     externalLinearDragV=<1 1>
    }
   }
   Save cached pose 'chuggaEmote'(Pose=Blend Poses by bool(
     BlendPose_0=Component To Local(ComponentPose=@n101)
     BlendPose_1=Blend Poses by bool(BlendPose_0=@n99, BlendPose_1=@n97, bActiveValue=@n100) {blendType=Linear}
     bActiveValue=Get SunburstEmote() -> (SunburstEmote:bool)
    ) {blendType=Linear}) {cacheName=chuggaEmote}
   Trail controller - Bone: dyn_cape_mid_r_06!(ComponentPose, Alpha=1, RelaxationSpeedScale:float=2, FakeVelocity:Vector="0,0,0") -> (Pose) {
    bActorSpaceFakeVel=true
    bShowTrailLocation=true
    bUsePlanarLimit=true
    chainLength=5
    debugLifeTime=0.1
    maxDeltaTime=0.0333
    relaxationSpeedScale=1.2
    alphaScaleBiasClamp {bInterpResult=true, interpSpeedDecreasing=25, interpSpeedIncreasing=25}
    baseJoint {boneName=root}
    rotationLimits [{limitMax=<180 180 180>}, {limitMax=<180 180 180>}, {limitMax=<180 180 180>}, {limitMax=<180 180 180>}, {limitMax=<180 180 180>}]
    rotationOffsets [<0 0 0>, <0 0 0>, <0 0 0>, <0 0 0>, <0 0 0>]
    trailBone {boneName=dyn_cape_mid_r_06}
    trailRelaxationSpeed {editorCurveData {keys [<at 0 15 RCIM_Linear>,<at 1 6 RCIM_Linear>]}}
   }
  ]}
 EventGraph {nodes [
   n1 = Event BlueprintUpdateAnimation()
   > n2 = Sequence!(execute) -> (
    then_0
    then_1
    then_2:Exec
    then_3:Exec
    then_4:Exec
    then_5:Exec
    then_6:Exec
    then_7:Exec
    then_8:Exec
    then_9:Exec
    then_10:Exec
    then_11:Exec
   )
   > n3 = Gravity Override State Function DEPRECATED(CharacterPartAnimInstance=Self-Reference(), GameTimeInSeconds=GetGameTimeInSeconds(), GravityOverrideParams=Get Gravity Override Params())
   > n4 = Set GravityOverride(GravityOverride=@n3)
   > n5 = Branch(Condition=Get bIsSkyDiving())
   > Set GravityOverrideLoincloth(GravityOverrideLoincloth:Vector=vector + vector(A:Vector=@n4, B:Vector="0, 0,-400") -> (ReturnValue:Vector) {is=title K2Node_CommutativeAssociativeBinaryOperator}) -> (Output_Get:Vector)
   Set GravityOverrideLoincloth(execute=@n5.else, GravityOverrideLoincloth=@n4)
   n6 = ComplexSimSpaceFunction DEPRECATED!(
    execute=@n2.then_1
    CharacterPartAnimInstance:Custom Character Part Anim Instance Object Reference=Self-Reference()
    InterpSpeedMax:double=9
    PawnSpeedMax:double=20
    RInterpErrorTolerance:double=1
    DeltaTime:double=@n1.DeltaTimeX
    DebugDraw:bool=false
    sim_space_rotation_in:Rotator=Get sim_space_rotation()
    inComponentSpaceIn:bool=Get InComponentSpace() -> (InComponentSpace:bool)
    inComponentSpaceLastFrameIn:bool=Get InComponentSpaceLastFrame() -> (InComponentSpaceLastFrame:bool)
    lockedOnSimSpaceIn:bool=Get LockedOnSimSpace() -> (LockedOnSimSpace:bool)
    InterpSpeedIn:double=Get InterpSpeed() -> (InterpSpeed:double)
   ) -> (
    then
    sim_space_rotation:Rotator
    inComponentSpaceOut:bool
    inComponentSpaceLastFrameOut:bool
    lockedOnSimSpaceOut:bool
    InterpSpeedOut:double
   )
   > Set sim_space_rotation(sim_space_rotation:Rotator=@n6) -> (Output_Get:Rotator)
   > Set InComponentSpace(InComponentSpace:bool=@n6.inComponentSpaceOut) -> (Output_Get:bool)
   > Set InComponentSpaceLastFrame(InComponentSpaceLastFrame:bool=@n6.inComponentSpaceLastFrameOut) -> (Output_Get:bool)
   > Set LockedOnSimSpace(LockedOnSimSpace:bool=@n6.lockedOnSimSpaceOut) -> (Output_Get:bool)
   > Set InterpSpeed(InterpSpeed:double=@n6.InterpSpeedOut) -> (Output_Get:double)
   n7 = PlayerMovement(execute=@n2.then_3, AnimInstance=Self-Reference(), XDotProductCoefficient=1, YDotProductCoefficient=1, ZDotProductCoefficient=1)
   n8 = MapRangeClamped(Value=@n7.YMovement, InRangeA=0, InRangeB=-1)
   n9 = "Absolute (Float)"(A=@n7)
   > Set BwdMovement(BwdMovement:double=MapRangeClamped(Value=float + float(A=@n9, B=@n8) {is=title K2Node_CommutativeAssociativeBinaryOperator}, InRangeA=0, InRangeB=1.4, OutRangeA=1, OutRangeB=0)) -> (Output_Get:double)
   n10 = BreakTransform(InTransform=GetDeltaTransformFromRefPose(self=GetOwnerSkeletalMeshComponent(), BoneName=thigh_r, BaseName=pelvis))
   Set thigh_rot_up_r(execute=@n2.then_5, thigh_rot_up_r:double=MapRangeClamped(Value=BreakRotator(InRot=@n10.Rotation).Yaw, InRangeA=0, InRangeB=-100)) -> (Output_Get:double)
   n11 = BreakTransform(InTransform=GetDeltaTransformFromRefPose(self=GetOwnerSkeletalMeshComponent(), BoneName=thigh_l, BaseName=pelvis))
   > Set thigh_rot_up_l(thigh_rot_up_l:double=MapRangeClamped(Value=BreakRotator(InRot=@n11.Rotation).Yaw, InRangeA=0, InRangeB=-100)) -> (Output_Get:double)
   n12 = Get bIsKneeSliding()
   > Set is_crouch_moving(is_crouch_moving:bool=AND Boolean!(A=Get bIsCrouching(), B=Get bIsMoving2D(), C:bool=NOT Boolean(A=@n12), D:bool=Get bRigidBodyEnabled()) -> (ReturnValue)) -> (Output_Get:bool)
   > n13 = Branch(Condition=Get bIsCrouching())
   > Set is_walk_moving(is_walk_moving:double=0) -> (Output_Get:double)
   n14 = Get thigh_rot_up_l()
   n15 = Get thigh_rot_up_r()
   Set is_walk_moving(execute=@n13.else, is_walk_moving="Max (Float)"(A=@n15, B=@n14))
   n16 = Get bRigidBodyEnabled()
   n17 = Get bIsCrouching()
   n18 = Branch(
    execute=@n2.then_10
    Condition=AND Boolean(A=Get bIsMoving2D(), B=Get bIsOnGround() -> (bIsOnGround:bool), C=NOT Boolean(A=@n17), D=NOT Boolean(A=@n16))
   )
   > Set Moving2D_On_Ground(Moving2D_On_Ground:double=1) -> (Output_Get:double)
   Set Moving2D_On_Ground(execute=@n18.else, Moving2D_On_Ground=0)
   n19 = Branch(execute=@n2.then_11, Condition=AND Boolean!(A=Get bIsBaseballSliding(), B=Get bRigidBodyEnabled()) -> (ReturnValue))
   > Set Is_BaseballSliding(Is_BaseballSliding:double=1) -> (Output_Get:double)
   Set Is_BaseballSliding(execute=@n19.else, Is_BaseballSliding=0)
   n20 = Break Fort Anim Input Slope Info!(FortAnimInput_SlopeInfo:"Fort Anim Input Slope Info Structure (by ref)"=Get SlopeInfoInput() -> (SlopeInfoInput:Fort Anim Input Slope Info Structure)) -> (RootSlopeTranslation:Vector=~, LeftFootSlopeTranslation:Vector=~, RightFootSlopeTranslation:Vector=~, SlopeRotation:Rotator=~) {is=title K2Node_BreakStruct}
   Set slope_root_trans(execute=@n2.then_2, slope_root_trans:Vector=@n20) -> (Output_Get:Vector)
   > Set slope_foot_rot(slope_foot_rot:Rotator=@n20.SlopeRotation) -> (Output_Get:Rotator)
   n21 = GetAnimInstance(self=GetOwnerSkeletalMeshComponent())
   Set chuggaEmote(
    execute=@n2.then_6
    chuggaEmote:bool=OR Boolean(
     A=Montage_IsActive!(
      self:Anim Instance Object Reference=@n21
      Montage:Anim Montage Object Reference=/BRCosmetics/Animation/Game/MainPlayer/Emotes/Chugga/Emote_Chugga_Train_CMF_Frontend_M
     ) -> (ReturnValue:bool)
     B=Montage_IsActive(self=@n21, Montage=/BRCosmetics/Animation/Game/MainPlayer/Emotes/Chugga/Emote_Chugga_Train_CMF_M)
    )
   ) -> (Output_Get:bool)
   n22 = GetAnimInstance(self=GetOwnerSkeletalMeshComponent())
   > Set SunburstEmote(SunburstEmote:bool=OR Boolean(
     A=Montage_IsPlaying(self=@n22, Montage=/BRCosmetics/Animation/Game/MainPlayer/Emotes/SunBurst/Emote_SunBurst_CMF_M_Frontend)
     B=Montage_IsPlaying(self=@n22, Montage=/BRCosmetics/Animation/Game/MainPlayer/Emotes/SunBurst/Emote_SunBurst_CMF_M)
    )) -> (Output_Get:bool)
   n23 = Branch(execute=@n2.then_4, Condition=Get bIsSwimSprinting() -> (bIsSwimSprinting:bool))
   > Set TrailControllerRot(TrailControllerRot:Rotator="15,0, 0") -> (Output_Get:Rotator)
   n24 = Branch(execute=@n23.else, Condition=Get bIsSkyDiving())
   > n25 = Branch(Condition=Get bIsParachuteOpened())
   > Set TrailControllerRot(TrailControllerRot="45,0, 0")
   n26 = Branch(execute=@n25.else, Condition="float > float"(A=Get PawnVelocityZ(), B=50) {is=title K2Node_CallFunction})
   > Set BwdMovement(BwdMovement=float * float(A=Get BwdMovement(), B=0))
   > Set TrailControllerRot(TrailControllerRot="45,0, 0")
   n27 = Branch(execute=@n26.else, Condition=Get bIsSkydiveDiveMode())
   > Set TrailControllerRot(TrailControllerRot="7,0, 0")
   Set TrailControllerRot(execute=@n27.else, TrailControllerRot="60,0, 0")
   Set TrailControllerRot(execute=@n24.else, TrailControllerRot="15,0, 0")
   > Set BwdMovement(BwdMovement=float * float(A=Get BwdMovement(), B=1))
   n28 = Get bRigidBodyEnabled()
   n29 = Get bIsCrouching()
   n30 = Branch(execute=@n2.then_7, Condition=AND Boolean!(A=Get bIsSprinting(), B=NOT Boolean(A=@n29), C:bool=NOT Boolean(A=@n28)) -> (ReturnValue))
   > Set is_SprintingNotSliding(is_SprintingNotSliding:double=1) -> (Output_Get:double)
   Set is_SprintingNotSliding(execute=@n30.else, is_SprintingNotSliding=0)
   n31 = Get bIsKneeSliding()
   n32 = Get bRigidBodyEnabled()
   n33 = Branch(
    execute=@n2.then_8
    Condition=AND Boolean!(A=Get bIsCrouching(), B=Get bIsMoving2D(), C=NOT Boolean(A=@n31), D:bool=NOT Boolean(A=@n32)) -> (ReturnValue)
   )
   > Set CrouchingMoving2D - Constraints(CrouchingMoving2D - Constraints:double=0.8) -> (Output_Get:double)
   > Set CrouchingMoving2D(CrouchingMoving2D:double=1) -> (Output_Get:double)
   Set CrouchingMoving2D - Constraints(execute=@n33.else, CrouchingMoving2D - Constraints=0)
   > Set CrouchingMoving2D(CrouchingMoving2D=0)
   n34 = Get bRigidBodyEnabled()
   n35 = Get bIsMoving2D()
   n36 = Branch(execute=@n2.then_9, Condition=AND Boolean!(A=Get bIsCrouching(), B=NOT Boolean(A=@n35), C=NOT Boolean(A=@n34)) -> (ReturnValue))
   > Set IddleCrouch(IddleCrouch:double=1) -> (Output_Get:double)
   Set IddleCrouch(execute=@n36.else, IddleCrouch=0)
   Event BlueprintInitializeAnimation()
   > Set SkelMesh(SkelMesh=GetPartSkeletalMeshComponent())
   Event OnBeginPlayingEmote() {is=title K2Node_Event}
   > Delay(Duration=0.2)
   > ResetDynamics(InTeleportType=TeleportPhysics)
   > ResetDynamics(InTeleportType=ResetPhysics)
   Event OnEndPlayingEmote() {is=title K2Node_Event}
   > Delay(Duration=0.2)
   > ResetDynamics(InTeleportType=TeleportPhysics)
   > ResetDynamics(InTeleportType=ResetPhysics)
   Event OnEndCrouching()
   > Set crouch alpha(crouch alpha:double=0) -> (Output_Get:double)
   > Set crouch alpha inv(crouch alpha inv:double=1) -> (Output_Get:double)
   Event OnCharacterCustomizationLoaded()
   n37 = CheckForBackbling()
   > Set isWearingBackbling(isWearingBackbling=OR Boolean(A=@n37, B=@n37.bIsBackblingLong))
   Event OnBeginCrouching()
   > Set crouch alpha(crouch alpha=1)
   > Set crouch alpha inv(crouch alpha inv=0)
   DrawJointDebugFunction!(
    execute
    SkelMeshComponent:Scene Component Object Reference
    JointOffset:Vector="0, 0, 0"
    JointName:Array of Names=Make Array("[0]":Name=dyn_root_simspace) -> (Array:Array of Names)
    Scale:double=10
    Duration:double=0
    Thickness:double=0.7
    DebugDraw:bool=true
    Foreground:EDrawDebugSceneDepthPriorityGroup Enum=Foreground
   ) -> (then)
   DrawVectorFromJointFunction!(
    execute
    SkelMeshComponent:Scene Component Object Reference
    JointName:Name=root
    LineEnd:Vector=@n4
    JointOffset:Vector="0, 0, 0"
    LineColor:Linear Color Structure=(R=1,G=0,B=0,A=1)
    Thickness:double=2
    Foreground:EDrawDebugSceneDepthPriorityGroup Enum=Foreground
   ) -> (then)
   float * float(A=@n15, B=@n14) {is=title K2Node_PromotableOperator}
  ]}
]
variables [
 Backward {type=number}
 BwdMovement {type=number}
 chuggaEmote {type=boolean}
 crouch alpha {type=number}
 crouch alpha inv {type=number, default=1}
 CrouchingMoving2D {type=number}
 CrouchingMoving2D - Constraints {type=number}
 Forward {type=number}
 Gravity Override Params {type=GravityOverrideParamsStruct}
 GravityOverride {type=Vector}
 GravityOverrideLoincloth {type=Vector}
 IddleCrouch {type=number}
 InComponentSpace {type=boolean}
 InComponentSpaceLastFrame {type=boolean}
 InterpSpeed {type=number}
 Is_BaseballSliding {type=number}
 is_crouch_moving {type=boolean}
 is_SprintingNotSliding {type=number}
 is_walk_moving {type=number}
 isWearingBackbling {type=boolean}
 Left {type=boolean}
 LockedOnSimSpace {type=boolean}
 Moving2D_On_Ground {type=number}
 Right {type=boolean}
 sim_space_rotation {type=Rotator}
 SkelMesh {type=SkeletalMeshComponent}
 slope_foot_rot {type=Rotator}
 slope_root_trans {type=Vector}
 SunburstEmote {type=boolean}
 thigh_rot_up_l {type=number}
 thigh_rot_up_r {type=number}
 TrailControllerRot {type=Rotator}
]
