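# Physics asset for the F_MED_Obsidian female player cosmetic, providing capsule bodies for its dynamic front hair and widow's peak.
# Left and right three-bone hair chains and a single widow's peak bone hang off the head, with narrow twist limits and asymmetric swing limits between sides.
# Head, neck, upper spine and upper-arm capsules form the surrounding body shape, linked by constraints with free swing and twist.
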PhysicsAsset /BRCosmetics/Characters/Player/Female/Medium/Bodies/F_MED_Obsidian/Meshes/Parts/F_MED_Obsidian_Dyn
bodies {
 dyn_hair_fr_l_01 [{shapeType=Capsule, length=3.6273, radius=1.3046, center=<2.3752 0 0>, rotation=<rot 90 0 0>}]
 dyn_hair_fr_l_02 [{shapeType=Capsule, length=3.7444, radius=1.1875, center=<2.3752 0.3764 0.0002>, rotation=<rot 90 0 0>}]
 dyn_hair_fr_l_03 [{shapeType=Capsule, length=6.0023, radius=1.1796, center=<2.5568 0.5247 0>, rotation=<rot 90 0 0>}]
 dyn_hair_fr_r_01 [{shapeType=Capsule, length=3.6273, radius=1.3046, center=<2.3752 0 0>, rotation=<rot 90 0 0>}]
 dyn_hair_fr_r_02 [{shapeType=Capsule, length=3.7444, radius=1.1875, center=<2.3752 -0.3764 0.0002>, rotation=<rot 90 0 0>}]
 dyn_hair_fr_r_03 [{shapeType=Capsule, length=6.0023, radius=1.1796, center=<2.5568 -0.5247 0>, rotation=<rot 90 0 0>}]
 dyn_widows_peak [{shapeType=Capsule, length=2.6898, radius=2.2421, center=<2.3752 0 0>, rotation=<rot 90 0 0>}]
 head [{shapeType=Capsule, length=8.6973, radius=4.7709, center=<-0.2916 3.5865 0.0001>, rotation=<rot 90 -90 -95.6982>}]
 neck_01 [{shapeType=Capsule, length=15.4062, radius=5.1562, center=<0.524 1.4232 0>, rotation=<rot 0 -5.0006 0>}]
 neck_02 [{shapeType=Capsule, length=15.4062, radius=5.1562, center=<1.0819 0.1513 0>, rotation=<rot 0 -5.0007 0>}]
 spine_05 [{shapeType=Capsule, length=21.041, radius=8.5521, center=<0.7168 -0.6132 0>}]
 upperarm_l [{shapeType=Capsule, length=24.8535, radius=4.7396, center=<14.4503 -0.6192 -0.1485>, rotation=<rot 90 0 0>}]
 upperarm_r [{shapeType=Capsule, length=24.8535, radius=4.7396, center=<-14.4503 0.6192 0.1485>, rotation=<rot 90 0 180>}]
}
constraints [
 {bone1Name=dyn_hair_fr_l_02, bone2Name=dyn_hair_fr_l_01, swing1LimitDegrees=30, swing2LimitDegrees=30, twistLimitDegrees=5}
 {bone1Name=dyn_hair_fr_l_03, bone2Name=dyn_hair_fr_l_02, swing1LimitDegrees=30, swing2LimitDegrees=30, twistLimitDegrees=5}
 {bone1Name=dyn_hair_fr_r_02, bone2Name=dyn_hair_fr_r_01, swing1LimitDegrees=15, swing2LimitDegrees=30, twistLimitDegrees=5}
 {bone1Name=dyn_hair_fr_r_03, bone2Name=dyn_hair_fr_r_02, swing1LimitDegrees=30, swing2LimitDegrees=30, twistLimitDegrees=5}
 {bone1Name=dyn_widows_peak, bone2Name=head, swing1LimitDegrees=5, swing2LimitDegrees=15, twistLimitDegrees=5}
 {bone1Name=dyn_hair_fr_l_01, bone2Name=head, swing1LimitDegrees=30, swing2LimitDegrees=30, twistLimitDegrees=5}
 {bone1Name=dyn_hair_fr_r_01, bone2Name=head, swing1LimitDegrees=15, swing2LimitDegrees=30, twistLimitDegrees=5}
 {
  bone1Name=head
  bone2Name=neck_02
  swing1LimitDegrees=45
  swing1Motion=Free
  swing2LimitDegrees=45
  swing2Motion=Free
  twistLimitDegrees=45
  twistMotion=Free
 }
 {
  bone1Name=neck_01
  bone2Name=spine_05
  swing1LimitDegrees=45
  swing1Motion=Free
  swing2LimitDegrees=45
  swing2Motion=Free
  twistLimitDegrees=45
  twistMotion=Free
 }
 {
  bone1Name=upperarm_l
  bone2Name=spine_05
  swing1LimitDegrees=45
  swing1Motion=Free
  swing2LimitDegrees=45
  swing2Motion=Free
  twistLimitDegrees=45
  twistMotion=Free
 }
 {
  bone1Name=upperarm_r
  bone2Name=spine_05
  swing1LimitDegrees=45
  swing1Motion=Free
  swing2LimitDegrees=45
  swing2Motion=Free
  twistLimitDegrees=45
  twistMotion=Free
 }
 {
  bone1Name=neck_02
  bone2Name=neck_01
  swing1LimitDegrees=45
  swing1Motion=Free
  swing2LimitDegrees=45
  swing2Motion=Free
  twistLimitDegrees=45
  twistMotion=Free
 }
]
